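# Physics asset for the cape of the F_MED_PoliteMoth medium female player cosmetic body.
# Four chains of four capsule bodies (inner and outer, left and right) hang from small base bodies, with inner bases angularly locked to spine_05 and outer bases to the upper arms.
# Chain joints share limit angles that are tight for swing1 and twist, wider for swing2.
# Pelvis, spine, clavicle and upper-arm capsules approximate the torso and arms.
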PhysicsAsset /BRCosmetics/Characters/Player/Female/Medium/Bodies/F_MED_PoliteMoth/Meshes/F_MED_PoliteMoth_Cape_Physics
bodies {
 base_cape_in_01_l [{shapeType=Capsule, length=0.505, radius=0.505, shapeName=base_cape_in_01_l_capsule, rotation=<rot 0 -90 0>}]
 base_cape_in_01_r [{shapeType=Capsule, length=0.505, radius=0.505, shapeName=base_cape_in_01_r_capsule, rotation=<rot 0 -90 0>}]
 base_cape_out_01_l [{shapeType=Capsule, length=0.505, radius=0.505, shapeName=base_cape_out_01_l_capsule, rotation=<rot 0 -90 0>}]
 base_cape_out_01_r [{shapeType=Capsule, length=0.505, radius=0.505, shapeName=base_cape_out_01_r_capsule, rotation=<rot 0 -90 0>}]
 clavicle_l [{shapeType=Capsule, length=8.4988, radius=7.8445, shapeName=clavicle_l_capsule, center=<3.8418 0.3662 -4.3521>, rotation=<rot 90 -173.6193 0>}]
 clavicle_r [{shapeType=Capsule, length=8.4988, radius=7.8445, shapeName=clavicle_r_capsule, center=<-3.8418 -0.3662 4.3521>, rotation=<rot 90 6.3807 0>}]
 dyn_cape_in_01_l [{shapeType=Capsule, length=9.2249, radius=5.5722, shapeName=dyn_cape_in_01_l_capsule, center=<6.4923 0 3.5>, rotation=<rot 0 0 -90>}]
 dyn_cape_in_01_r [{shapeType=Capsule, length=9.2249, radius=5.5722, shapeName=dyn_cape_in_01_r_capsule, center=<6.4923 0 3.5>, rotation=<rot 0 0 90>}]
 dyn_cape_in_02_l [{shapeType=Capsule, length=9.2249, radius=5.5722, shapeName=dyn_cape_in_02_l_capsule, center=<6.4923 0 3.5>, rotation=<rot 0 0 90>}]
 dyn_cape_in_02_r [{shapeType=Capsule, length=9.2249, radius=5.5722, shapeName=dyn_cape_in_02_r_capsule, center=<6.4923 0 3.5>, rotation=<rot 0 0 90>}]
 dyn_cape_in_03_l [{shapeType=Capsule, length=9.2249, radius=5.5722, shapeName=dyn_cape_in_03_l_capsule, center=<6.4923 0 3.5>, rotation=<rot 0 0 90>}]
 dyn_cape_in_03_r [{shapeType=Capsule, length=9.2249, radius=5.5722, shapeName=dyn_cape_in_03_r_capsule, center=<6.4923 0 3.5>, rotation=<rot 0 0 90>}]
 dyn_cape_in_04_l [{shapeType=Capsule, length=9.2249, radius=5.5722, shapeName=dyn_cape_in_04_l_capsule, center=<6.4923 0 3.5>, rotation=<rot 0 0 90>}]
 dyn_cape_in_04_r [{shapeType=Capsule, length=9.2249, radius=5.5722, shapeName=dyn_cape_in_04_r_capsule, center=<6.4923 0 3.5>, rotation=<rot 0 0 90>}]
 dyn_cape_out_01_l [{
   shapeType=Capsule
   length=3.1458
   radius=4.7522
   shapeName=dyn_cape_out_01_l_capsule
   center=<6.4923 0 2.5>
   rotation=<rot -0.0001 0 100.7907>
  }]
 dyn_cape_out_01_r [{
   shapeType=Capsule
   length=3.1458
   radius=4.7522
   shapeName=dyn_cape_out_01_r_capsule
   center=<6.4923 0 2.5>
   rotation=<rot -0.0001 0 76.6178>
  }]
 dyn_cape_out_02_l [{shapeType=Capsule, length=5.6885, radius=5.5722, shapeName=dyn_cape_out_02_l_capsule, center=<6.4923 0 2.5>, rotation=<rot 0 0 90>}]
 dyn_cape_out_02_r [{shapeType=Capsule, length=5.6885, radius=5.5722, shapeName=dyn_cape_out_02_r_capsule, center=<6.4923 0 2.5>, rotation=<rot 0 0 90>}]
 dyn_cape_out_03_l [{shapeType=Capsule, length=9.2249, radius=5.5722, shapeName=dyn_cape_out_03_l_capsule, center=<6.4923 0 2.5>, rotation=<rot 0 0 90>}]
 dyn_cape_out_03_r [{shapeType=Capsule, length=9.2249, radius=5.5722, shapeName=dyn_cape_out_03_r_capsule, center=<6.4923 0 2.5>, rotation=<rot 0 0 90>}]
 dyn_cape_out_04_l [{shapeType=Capsule, length=13.4004, radius=5.5722, shapeName=dyn_cape_out_04_l_capsule, center=<6.4923 0 2.5>, rotation=<rot 0 0 90>}]
 dyn_cape_out_04_r [{shapeType=Capsule, length=13.4004, radius=5.5722, shapeName=dyn_cape_out_04_r_capsule, center=<6.4923 0 2.5>, rotation=<rot 0 0 90>}]
 pelvis [{shapeType=Capsule, length=11.376, radius=10.2596, shapeName=pelvis_capsule, center=<-7.634 -1.9985 -0.7183>, rotation=<rot 0 -90 0>}]
 spine_01 [{shapeType=Capsule, length=10.9957, radius=10.2596, shapeName=pelvis_capsule, center=<-7.634 -0.0881 -0.7183>, rotation=<rot 0 -90 0>}]
 spine_02 [{shapeType=Capsule, length=5.5295, radius=10.2596, shapeName=pelvis_capsule, center=<-7.634 1.5324 -0.7183>, rotation=<rot 0 -90 0>}]
 spine_03 [{shapeType=Capsule, length=4.3091, radius=9.3038, shapeName=pelvis_capsule, center=<-7.634 1.5324 -0.7183>, rotation=<rot 0 -90 0>}]
 spine_04 [{shapeType=Capsule, length=5.7434, radius=8.4964, shapeName=pelvis_capsule, center=<-7.634 1.5324 -0.7183>, rotation=<rot 0 -90 0>}]
 spine_05 [{shapeType=Capsule, length=5.5826, radius=11.0896, shapeName=pelvis_capsule, center=<-5.1312 1.5324 -0.7183>, rotation=<rot 0 -90 0>}]
 upperarm_l [{shapeType=Capsule, length=26.6916, radius=4.0683, shapeName=upperarm_l_capsule, center=<12.4869 -0.2242 -1.209>, rotation=<rot 90 180 0>}]
 upperarm_r [{shapeType=Capsule, length=26.6916, radius=4.0683, shapeName=upperarm_r_capsule, center=<-12.4869 0.2242 1.209>, rotation=<rot 90 0 0>}]
}
constraints [
 {
  bone1Name=base_cape_in_01_l
  bone2Name=spine_05
  swing1LimitDegrees=45
  swing1Motion=Locked
  swing2LimitDegrees=45
  swing2Motion=Locked
  twistLimitDegrees=45
  twistMotion=Locked
 }
 {bone1Name=dyn_cape_in_01_l, bone2Name=base_cape_in_01_l, swing1LimitDegrees=5, swing2LimitDegrees=30, twistLimitDegrees=5}
 {bone1Name=dyn_cape_in_02_l, bone2Name=dyn_cape_in_01_l, swing1LimitDegrees=5, swing2LimitDegrees=30, twistLimitDegrees=5}
 {bone1Name=dyn_cape_in_03_l, bone2Name=dyn_cape_in_02_l, swing1LimitDegrees=5, swing2LimitDegrees=30, twistLimitDegrees=5}
 {bone1Name=dyn_cape_in_04_l, bone2Name=dyn_cape_in_03_l, swing1LimitDegrees=5, swing2LimitDegrees=30, twistLimitDegrees=5}
 {
  bone1Name=base_cape_in_01_r
  bone2Name=spine_05
  swing1LimitDegrees=45
  swing1Motion=Locked
  swing2LimitDegrees=45
  swing2Motion=Locked
  twistLimitDegrees=45
  twistMotion=Locked
 }
 {bone1Name=dyn_cape_in_01_r, bone2Name=base_cape_in_01_r, swing1LimitDegrees=5, swing2LimitDegrees=30, twistLimitDegrees=5}
 {bone1Name=dyn_cape_in_02_r, bone2Name=dyn_cape_in_01_r, swing1LimitDegrees=5, swing2LimitDegrees=30, twistLimitDegrees=5}
 {bone1Name=dyn_cape_in_03_r, bone2Name=dyn_cape_in_02_r, swing1LimitDegrees=5, swing2LimitDegrees=30, twistLimitDegrees=5}
 {bone1Name=dyn_cape_in_04_r, bone2Name=dyn_cape_in_03_r, swing1LimitDegrees=5, swing2LimitDegrees=30, twistLimitDegrees=5}
 {
  bone1Name=base_cape_out_01_l
  bone2Name=upperarm_l
  swing1LimitDegrees=45
  swing1Motion=Locked
  swing2LimitDegrees=45
  swing2Motion=Locked
  twistLimitDegrees=45
  twistMotion=Locked
 }
 {bone1Name=dyn_cape_out_01_l, bone2Name=base_cape_out_01_l, swing1LimitDegrees=5, swing2LimitDegrees=30, twistLimitDegrees=5}
 {bone1Name=dyn_cape_out_02_l, bone2Name=dyn_cape_out_01_l, swing1LimitDegrees=5, swing2LimitDegrees=30, twistLimitDegrees=5}
 {bone1Name=dyn_cape_out_03_l, bone2Name=dyn_cape_out_02_l, swing1LimitDegrees=5, swing2LimitDegrees=30, twistLimitDegrees=5}
 {bone1Name=dyn_cape_out_04_l, bone2Name=dyn_cape_out_03_l, swing1LimitDegrees=5, swing2LimitDegrees=30, twistLimitDegrees=5}
 {
  bone1Name=base_cape_out_01_r
  bone2Name=upperarm_r
  swing1LimitDegrees=45
  swing1Motion=Locked
  swing2LimitDegrees=45
  swing2Motion=Locked
  twistLimitDegrees=45
  twistMotion=Locked
 }
 {bone1Name=dyn_cape_out_01_r, bone2Name=base_cape_out_01_r, swing1LimitDegrees=5, swing2LimitDegrees=30, twistLimitDegrees=5}
 {bone1Name=dyn_cape_out_02_r, bone2Name=dyn_cape_out_01_r, swing1LimitDegrees=5, swing2LimitDegrees=30, twistLimitDegrees=5}
 {bone1Name=dyn_cape_out_03_r, bone2Name=dyn_cape_out_02_r, swing1LimitDegrees=5, swing2LimitDegrees=30, twistLimitDegrees=5}
 {bone1Name=dyn_cape_out_04_r, bone2Name=dyn_cape_out_03_r, swing1LimitDegrees=5, swing2LimitDegrees=30, twistLimitDegrees=5}
]
