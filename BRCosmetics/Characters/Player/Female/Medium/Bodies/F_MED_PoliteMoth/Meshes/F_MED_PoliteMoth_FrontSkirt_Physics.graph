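# Physics asset for the skirt of the PoliteMoth medium female body cosmetic, covering front, side and back strands despite its FrontSkirt name.
# Seven strands each run from a pelvis-linked base through five capsule bodies with limited swing and twist.
# Cross constraints with free rotation pair the left and right front strands from the second link down.
# Capsules also cover pelvis, legs and feet, plus a huge root capsule sitting mostly below the origin.
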PhysicsAsset /BRCosmetics/Characters/Player/Female/Medium/Bodies/F_MED_PoliteMoth/Meshes/F_MED_PoliteMoth_FrontSkirt_Physics
bodies {
 base_skirt_01_bck [{shapeType=Capsule, length=0.505, radius=0.505, shapeName=base_skirt_01_bck_capsule, rotation=<rot 0 -90 0>}]
 base_skirt_01_bck_l [{shapeType=Capsule, length=0.505, radius=0.505, shapeName=base_skirt_01_bck_l_capsule, rotation=<rot 0 -90 0>}]
 base_skirt_01_bck_r [{shapeType=Capsule, length=0.505, radius=0.505, shapeName=base_skirt_01_bck_r_capsule, rotation=<rot 0 -90 0>}]
 base_skirt_01_front_l [{shapeType=Capsule, length=0.505, radius=0.505, shapeName=base_skirt_01_front_l_capsule, rotation=<rot 0 -90 0>}]
 base_skirt_01_front_r [{shapeType=Capsule, length=0.505, radius=0.505, shapeName=base_skirt_01_front_r_capsule, rotation=<rot 0 -90 0>}]
 base_skirt_01_l [{shapeType=Capsule, length=0.505, radius=0.505, shapeName=base_skirt_01_l_capsule, rotation=<rot 0 -90 0>}]
 base_skirt_01_r [{shapeType=Capsule, length=0.505, radius=0.505, shapeName=base_skirt_01_r_capsule, rotation=<rot 0 -90 0>}]
 calf_l [{shapeType=Capsule, length=44.0305, radius=6.3942, shapeName=calf_l_capsule, center=<-22.8717 -1.8771 -0.1211>, rotation=<rot 90 0 0>}]
 calf_r [{shapeType=Capsule, length=44.0305, radius=6.3942, shapeName=calf_r_capsule, center=<22.8717 1.8771 0.1211>, rotation=<rot 90 180 0>}]
 dyn_skirt_01_bck [{shapeType=Capsule, length=10, radius=10, shapeName=dyn_skirt_01_bck_capsule, center=<8.4683 0 8>, rotation=<rot 0 0 90>}]
 dyn_skirt_01_bck_l [{shapeType=Capsule, length=10, radius=10, shapeName=dyn_skirt_01_bck_l_capsule, center=<8.4683 0 8>, rotation=<rot 0 0 90>}]
 dyn_skirt_01_bck_r [{shapeType=Capsule, length=10, radius=10, shapeName=dyn_skirt_01_bck_r_capsule, center=<8.4683 0 8>, rotation=<rot 0 0 90>}]
 dyn_skirt_01_front_l [{shapeType=Capsule, length=2.5, radius=4, shapeName=dyn_skirt_01_front_l_capsule, center=<5.8914 0 2.3323>, rotation=<rot 0 0 90>}]
 dyn_skirt_01_front_r [{shapeType=Capsule, length=2.5, radius=4, shapeName=dyn_skirt_01_front_r_capsule, center=<5.8914 0 1.0183>, rotation=<rot 0 0 -90>}]
 dyn_skirt_01_l [{shapeType=Capsule, length=10, radius=10, shapeName=dyn_skirt_01_l_capsule, center=<8.4683 0 8>, rotation=<rot 0 0 90>}]
 dyn_skirt_01_r [{shapeType=Capsule, length=10, radius=10, shapeName=dyn_skirt_01_r_capsule, center=<8.4683 0 8>, rotation=<rot 0 0 90>}]
 dyn_skirt_02_bck [{shapeType=Capsule, length=10, radius=10, shapeName=dyn_skirt_02_bck_capsule, center=<8.4683 0 8>, rotation=<rot 0 0 90>}]
 dyn_skirt_02_bck_l [{shapeType=Capsule, length=10, radius=10, shapeName=dyn_skirt_02_bck_l_capsule, center=<8.4683 0 8>, rotation=<rot 0 0 90>}]
 dyn_skirt_02_bck_r [{shapeType=Capsule, length=10, radius=10, shapeName=dyn_skirt_02_bck_r_capsule, center=<8.4683 0 8>, rotation=<rot 0 0 90>}]
 dyn_skirt_02_front_l [{shapeType=Capsule, length=0.8911, radius=4, shapeName=dyn_skirt_02_front_l_capsule, center=<5.4509 0 0>, rotation=<rot 0 0 90>}]
 dyn_skirt_02_front_r [{shapeType=Capsule, length=2.5, radius=4, shapeName=dyn_skirt_02_front_r_capsule, center=<5.4509 0 0>, rotation=<rot 0 0 -90>}]
 dyn_skirt_02_l [{shapeType=Capsule, length=10, radius=10, shapeName=dyn_skirt_02_l_capsule, center=<8.4683 0 8>, rotation=<rot 0 0 90>}]
 dyn_skirt_02_r [{shapeType=Capsule, length=10, radius=10, shapeName=dyn_skirt_02_r_capsule, center=<8.4683 0 8>, rotation=<rot 0 -0.0001 75.5659>}]
 dyn_skirt_03_bck [{shapeType=Capsule, length=10, radius=10, shapeName=dyn_skirt_03_bck_capsule, center=<8.4683 0 8>, rotation=<rot 0 0 90>}]
 dyn_skirt_03_bck_l [{shapeType=Capsule, length=10, radius=10, shapeName=dyn_skirt_03_bck_l_capsule, center=<8.4683 0 8>, rotation=<rot 0 0.1444 90>}]
 dyn_skirt_03_bck_r [{shapeType=Capsule, length=10, radius=10, shapeName=dyn_skirt_03_bck_r_capsule, center=<8.4683 0 8>, rotation=<rot 0 0 90>}]
 dyn_skirt_03_front_l [{shapeType=Capsule, length=2.5, radius=4, shapeName=dyn_skirt_03_front_l_capsule, center=<3.0754 0 0>, rotation=<rot 0 0 90>}]
 dyn_skirt_03_front_r [{shapeType=Capsule, length=2.5, radius=4, shapeName=dyn_skirt_03_front_r_capsule, center=<3.0754 0 0>, rotation=<rot 0 0 -90>}]
 dyn_skirt_03_l [{shapeType=Capsule, length=10, radius=10, shapeName=dyn_skirt_03_l_capsule, center=<8.4683 0 8>, rotation=<rot 0 0 90>}]
 dyn_skirt_03_r [{shapeType=Capsule, length=10, radius=10, shapeName=dyn_skirt_03_r_capsule, center=<8.4683 0 8>, rotation=<rot 0 0 90>}]
 dyn_skirt_04_bck [{shapeType=Capsule, length=10, radius=10, shapeName=dyn_skirt_04_bck_capsule, center=<8.4683 0 8>, rotation=<rot 0 0 90>}]
 dyn_skirt_04_bck_l [{shapeType=Capsule, length=10, radius=10, shapeName=dyn_skirt_04_bck_l_capsule, center=<8.4683 0 8>, rotation=<rot 0 0.1444 90>}]
 dyn_skirt_04_bck_r [{shapeType=Capsule, length=10, radius=10, shapeName=dyn_skirt_04_bck_r_capsule, center=<8.4683 0 8>, rotation=<rot 0 0 90>}]
 dyn_skirt_04_front_l [{shapeType=Capsule, length=2.5, radius=4, shapeName=dyn_skirt_04_front_l_capsule, center=<5.9954 0 0>, rotation=<rot 0 0 90>}]
 dyn_skirt_04_front_r [{shapeType=Capsule, length=2.5, radius=4, shapeName=dyn_skirt_04_front_r_capsule, center=<5.9954 0 0>, rotation=<rot 0 0 -90>}]
 dyn_skirt_04_l [{shapeType=Capsule, length=10, radius=10, shapeName=dyn_skirt_04_l_capsule, center=<8.4683 0 8>, rotation=<rot 0 0 90>}]
 dyn_skirt_04_r [{shapeType=Capsule, length=10, radius=10, shapeName=dyn_skirt_04_r_capsule, center=<8.4683 0 8>, rotation=<rot 0 0 90>}]
 dyn_skirt_05_bck [{shapeType=Capsule, length=10, radius=10, shapeName=dyn_skirt_05_bck_capsule, center=<8.4683 0 8>, rotation=<rot 0 0 90>}]
 dyn_skirt_05_bck_l [{shapeType=Capsule, length=10, radius=10, shapeName=dyn_skirt_05_bck_l_capsule, center=<8.4683 0 8>, rotation=<rot 0 0.1444 90>}]
 dyn_skirt_05_bck_r [{shapeType=Capsule, length=10, radius=10, shapeName=dyn_skirt_05_bck_r_capsule, center=<8.4683 0 8>, rotation=<rot 0 0 90>}]
 dyn_skirt_05_front_l [{shapeType=Capsule, length=2.5, radius=4, shapeName=dyn_skirt_05_front_l_capsule, center=<3.0754 0 0>, rotation=<rot 0 0 90>}]
 dyn_skirt_05_front_r [{shapeType=Capsule, length=2.5, radius=4, shapeName=dyn_skirt_05_front_r_capsule, center=<3.0754 0 0>, rotation=<rot 0 0 -90>}]
 dyn_skirt_05_l [{shapeType=Capsule, length=10, radius=10, shapeName=dyn_skirt_05_l_capsule, center=<8.4683 0 8>, rotation=<rot 0 0 90>}]
 dyn_skirt_05_r [{shapeType=Capsule, length=10, radius=10, shapeName=dyn_skirt_05_r_capsule, center=<8.4683 0 8>, rotation=<rot 0 0 90>}]
 foot_l [{
   shapeType=Capsule
   length=9.3057
   radius=9.0734
   shapeName=foot_l_capsule
   center=<0.0663 4.1205 0.0631>
   rotation=<rot -1.9866 -179.1916 -91.9204>
  }]
 foot_r [{
   shapeType=Capsule
   length=9.3057
   radius=9.0734
   shapeName=foot_r_capsule
   center=<-0.0663 -4.1205 -0.0631>
   rotation=<rot -1.9866 -179.1916 88.0796>
  }]
 pelvis [{shapeType=Capsule, length=19.0987, radius=10.2596, shapeName=pelvis_capsule, center=<-7.634 1.5324 -0.7183>, rotation=<rot 0 -90 0>}]
 root [{shapeType=Capsule, length=285.4152, radius=243.3284, shapeName=root_capsule, center=<0 0 -234.8617>, rotation=<rot 90 -90 0>}]
 thigh_l [{shapeType=Capsule, length=28.1565, radius=7, shapeName=thigh_l_capsule, center=<-20.8181 -0.6484 -1.3698>, rotation=<rot 90 0 0>}]
 thigh_r [{shapeType=Capsule, length=28.1565, radius=7, shapeName=thigh_r_capsule, center=<20.8181 0.6484 1.3698>, rotation=<rot 90 180 0>}]
}
constraints [
 {bone1Name=dyn_skirt_05_front_r, bone2Name=dyn_skirt_04_front_r, swing1LimitDegrees=20, swing2LimitDegrees=30, twistLimitDegrees=5}
 {
  bone1Name=base_skirt_01_front_r
  bone2Name=pelvis
  swing1LimitDegrees=45
  swing1Motion=Free
  swing2LimitDegrees=45
  swing2Motion=Free
  twistLimitDegrees=45
  twistMotion=Free
 }
 {bone1Name=dyn_skirt_01_front_r, bone2Name=base_skirt_01_front_r, swing1LimitDegrees=10, swing2LimitDegrees=20, twistLimitDegrees=5}
 {bone1Name=dyn_skirt_02_front_r, bone2Name=dyn_skirt_01_front_r, swing1LimitDegrees=20, swing2LimitDegrees=40, twistLimitDegrees=5}
 {bone1Name=dyn_skirt_03_front_r, bone2Name=dyn_skirt_02_front_r, swing1LimitDegrees=20, swing2LimitDegrees=30, twistLimitDegrees=5}
 {bone1Name=dyn_skirt_04_front_r, bone2Name=dyn_skirt_03_front_r, swing1LimitDegrees=20, swing2LimitDegrees=30, twistLimitDegrees=5}
 {bone1Name=dyn_skirt_05_front_l, bone2Name=dyn_skirt_04_front_l, swing1LimitDegrees=20, swing2LimitDegrees=30, twistLimitDegrees=5}
 {
  bone1Name=base_skirt_01_front_l
  bone2Name=pelvis
  swing1LimitDegrees=45
  swing1Motion=Free
  swing2LimitDegrees=45
  swing2Motion=Free
  twistLimitDegrees=45
  twistMotion=Free
 }
 {bone1Name=dyn_skirt_01_front_l, bone2Name=base_skirt_01_front_l, swing1LimitDegrees=10, swing2LimitDegrees=20, twistLimitDegrees=5}
 {bone1Name=dyn_skirt_02_front_l, bone2Name=dyn_skirt_01_front_l, swing1LimitDegrees=20, swing2LimitDegrees=40, twistLimitDegrees=5}
 {bone1Name=dyn_skirt_03_front_l, bone2Name=dyn_skirt_02_front_l, swing1LimitDegrees=20, swing2LimitDegrees=30, twistLimitDegrees=5}
 {bone1Name=dyn_skirt_04_front_l, bone2Name=dyn_skirt_03_front_l, swing1LimitDegrees=20, swing2LimitDegrees=30, twistLimitDegrees=5}
 {bone1Name=dyn_skirt_05_r, bone2Name=dyn_skirt_04_r, swing1LimitDegrees=10, swing2LimitDegrees=50, twistLimitDegrees=5}
 {
  bone1Name=base_skirt_01_l
  bone2Name=pelvis
  swing1LimitDegrees=45
  swing1Motion=Free
  swing2LimitDegrees=45
  swing2Motion=Free
  twistLimitDegrees=45
  twistMotion=Free
 }
 {bone1Name=dyn_skirt_01_l, bone2Name=base_skirt_01_l, swing1LimitDegrees=5, swing2LimitDegrees=20, twistLimitDegrees=5}
 {bone1Name=dyn_skirt_02_l, bone2Name=dyn_skirt_01_l, swing1LimitDegrees=10, swing2LimitDegrees=50, twistLimitDegrees=5}
 {bone1Name=dyn_skirt_03_l, bone2Name=dyn_skirt_02_l, swing1LimitDegrees=10, swing2LimitDegrees=50, twistLimitDegrees=5}
 {bone1Name=dyn_skirt_04_l, bone2Name=dyn_skirt_03_l, swing1LimitDegrees=10, swing2LimitDegrees=50, twistLimitDegrees=5}
 {bone1Name=dyn_skirt_05_l, bone2Name=dyn_skirt_04_l, swing1LimitDegrees=10, swing2LimitDegrees=50, twistLimitDegrees=5}
 {
  bone1Name=base_skirt_01_r
  bone2Name=pelvis
  swing1LimitDegrees=45
  swing1Motion=Free
  swing2LimitDegrees=45
  swing2Motion=Free
  twistLimitDegrees=45
  twistMotion=Free
 }
 {bone1Name=dyn_skirt_01_r, bone2Name=base_skirt_01_r, swing1LimitDegrees=5, swing2LimitDegrees=20, twistLimitDegrees=5}
 {bone1Name=dyn_skirt_02_r, bone2Name=dyn_skirt_01_r, swing1LimitDegrees=10, swing2LimitDegrees=50, twistLimitDegrees=15}
 {bone1Name=dyn_skirt_03_r, bone2Name=dyn_skirt_02_r, swing1LimitDegrees=10, swing2LimitDegrees=50, twistLimitDegrees=5}
 {bone1Name=dyn_skirt_04_r, bone2Name=dyn_skirt_03_r, swing1LimitDegrees=10, swing2LimitDegrees=50, twistLimitDegrees=5}
 {bone1Name=dyn_skirt_05_bck_r, bone2Name=dyn_skirt_04_bck_r, swing1LimitDegrees=10, swing2LimitDegrees=40, twistLimitDegrees=5}
 {
  bone1Name=base_skirt_01_bck
  bone2Name=pelvis
  swing1LimitDegrees=45
  swing1Motion=Free
  swing2LimitDegrees=45
  swing2Motion=Free
  twistLimitDegrees=45
  twistMotion=Free
 }
 {bone1Name=dyn_skirt_01_bck, bone2Name=base_skirt_01_bck, swing1LimitDegrees=5, swing2LimitDegrees=20, twistLimitDegrees=5}
 {bone1Name=dyn_skirt_02_bck, bone2Name=dyn_skirt_01_bck, swing1LimitDegrees=10, swing2LimitDegrees=50, twistLimitDegrees=5}
 {bone1Name=dyn_skirt_03_bck, bone2Name=dyn_skirt_02_bck, swing1LimitDegrees=10, swing2LimitDegrees=40, twistLimitDegrees=5}
 {bone1Name=dyn_skirt_04_bck, bone2Name=dyn_skirt_03_bck, swing1LimitDegrees=10, swing2LimitDegrees=40, twistLimitDegrees=5}
 {bone1Name=dyn_skirt_05_bck, bone2Name=dyn_skirt_04_bck, swing1LimitDegrees=10, swing2LimitDegrees=40, twistLimitDegrees=5}
 {
  bone1Name=base_skirt_01_bck_l
  bone2Name=pelvis
  swing1LimitDegrees=45
  swing1Motion=Free
  swing2LimitDegrees=45
  swing2Motion=Free
  twistLimitDegrees=45
  twistMotion=Free
 }
 {bone1Name=dyn_skirt_01_bck_l, bone2Name=base_skirt_01_bck_l, swing1LimitDegrees=5, swing2LimitDegrees=20, twistLimitDegrees=5}
 {bone1Name=dyn_skirt_02_bck_l, bone2Name=dyn_skirt_01_bck_l, swing1LimitDegrees=10, swing2LimitDegrees=50, twistLimitDegrees=5}
 {bone1Name=dyn_skirt_03_bck_l, bone2Name=dyn_skirt_02_bck_l, swing1LimitDegrees=10, swing2LimitDegrees=40, twistLimitDegrees=5}
 {bone1Name=dyn_skirt_04_bck_l, bone2Name=dyn_skirt_03_bck_l, swing1LimitDegrees=10, swing2LimitDegrees=40, twistLimitDegrees=5}
 {bone1Name=dyn_skirt_05_bck_l, bone2Name=dyn_skirt_04_bck_l, swing1LimitDegrees=10, swing2LimitDegrees=40, twistLimitDegrees=5}
 {
  bone1Name=base_skirt_01_bck_r
  bone2Name=pelvis
  swing1LimitDegrees=45
  swing1Motion=Free
  swing2LimitDegrees=45
  swing2Motion=Free
  twistLimitDegrees=45
  twistMotion=Free
 }
 {bone1Name=dyn_skirt_01_bck_r, bone2Name=base_skirt_01_bck_r, swing1LimitDegrees=5, swing2LimitDegrees=20, twistLimitDegrees=5}
 {bone1Name=dyn_skirt_02_bck_r, bone2Name=dyn_skirt_01_bck_r, swing1LimitDegrees=10, swing2LimitDegrees=50, twistLimitDegrees=5}
 {bone1Name=dyn_skirt_03_bck_r, bone2Name=dyn_skirt_02_bck_r, swing1LimitDegrees=10, swing2LimitDegrees=40, twistLimitDegrees=5}
 {bone1Name=dyn_skirt_04_bck_r, bone2Name=dyn_skirt_03_bck_r, swing1LimitDegrees=10, swing2LimitDegrees=40, twistLimitDegrees=5}
 {
  bone1Name=dyn_skirt_02_front_r
  bone2Name=dyn_skirt_02_front_l
  swing1LimitDegrees=45
  swing1Motion=Free
  swing2LimitDegrees=45
  swing2Motion=Free
  twistLimitDegrees=45
  twistMotion=Free
 }
 {
  bone1Name=dyn_skirt_03_front_r
  bone2Name=dyn_skirt_03_front_l
  swing1LimitDegrees=45
  swing1Motion=Free
  swing2LimitDegrees=45
  swing2Motion=Free
  twistLimitDegrees=45
  twistMotion=Free
 }
 {
  bone1Name=dyn_skirt_04_front_r
  bone2Name=dyn_skirt_04_front_l
  swing1LimitDegrees=45
  swing1Motion=Free
  swing2LimitDegrees=45
  swing2Motion=Free
  twistLimitDegrees=45
  twistMotion=Free
 }
 {
  bone1Name=dyn_skirt_05_front_r
  bone2Name=dyn_skirt_05_front_l
  swing1LimitDegrees=45
  swing1Motion=Free
  swing2LimitDegrees=45
  swing2Motion=Free
  twistLimitDegrees=45
  twistMotion=Free
 }
]
