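# Control Rig for the PoliteMoth female character body that procedurally corrects skirt, belt, cape and bracelet bones as the body moves.
# FortDeformRig is paired with Skirt Setup, Cape Setup and Variables functions, which use deform alpha attributes and remapped joint rotations as corrective weights.
# Skirt Setup also uses spherical pose readers and gameplay states like crouching, knee sliding, DBNO, emoting and skydiving.
# Weighted parent constraints anchor skirt and bracelet bones.
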ControlRigBlueprint /BRCosmetics/Characters/Player/Female/Medium/Bodies/F_MED_PoliteMoth/Meshes/M_MED_PoliteMoth_CtrlRig
graphs [
 Multiply_2_2_ContainedGraph {nodes [
   n1 = RigVMFunctionEntryNode()
   RigVMFunctionReturnNode(Result=&RigVMFunction_MathDoubleMul(B=@n1.C, A=&RigVMFunction_MathDoubleMul(A=@n1.A, B=@n1.B).Result).Result)
  ]}
 Multiply_3_1_ContainedGraph {nodes [
   n1 = RigVMFunctionEntryNode()
   RigVMFunctionReturnNode(Result=&RigVMFunction_MathDoubleMul(B=@n1.C, A=&RigVMFunction_MathDoubleMul(A=@n1.A, B=@n1.B).Result).Result)
  ]}
 New Function_ContainedGraph {nodes [
   n1 = RigVMFunctionEntryNode()
   RigVMFunction_MathQuaternionToEuler = &RigVMFunction_MathQuaternionToEuler(
    RotationOrder=XYZ
    Value=&RigUnit_GetRelativeTransformForItem(
     Child=(Type=Bone,Name="hand_l")
     Parent=(Type=Null,Name="lowerarm_l")
     bChildInitial=False
     bParentInitial=False
     Child.Name=@n1.Child
     Parent.Name=@n1.Parent
    ).RelativeTransform.Rotation
   )
   Remap = &RigVMFunction_MathDoubleRemap(
    TargetMaximum=1
    TargetMinimum=0
    Value=@RigVMFunction_MathQuaternionToEuler.Result.Y
    SourceMaximum=@n1.Source Maximum
    SourceMinimum=@n1.Source Minimum
   )
   > RigVMFunctionReturnNode(Delta Rotation=@Remap.Result)
  ]}
 New Function_ContainedGraph {nodes [
   n1 = RigVMFunctionEntryNode()
   RigVMFunction_MathQuaternionToEuler = &RigVMFunction_MathQuaternionToEuler(
    RotationOrder=XYZ
    Value=&RigUnit_GetRelativeTransformForItem(
     Child=(Type=Bone,Name="hand_l")
     Parent=(Type=Null,Name="lowerarm_l")
     bChildInitial=False
     bParentInitial=False
     Child.Name=@n1.Child
     Parent.Name=@n1.Parent
    ).RelativeTransform.Rotation
   )
   Remap = &RigVMFunction_MathDoubleRemap(
    TargetMaximum=1
    TargetMinimum=0
    Value=@RigVMFunction_MathQuaternionToEuler.Result.Z
    SourceMaximum=@n1.Source Maximum
    SourceMinimum=@n1.Source Minimum
   )
   > RigVMFunctionReturnNode(Delta Rotation=@Remap.Result)
  ]}
 New Function_ContainedGraph {nodes [
   n1 = RigVMFunctionEntryNode()
   RigVMFunction_MathTransformToEulerTransform = /Script/RigVM.RigVMFunction_MathTransformToEulerTransform(Value=&RigVMFunction_MathTransformMakeRelative(
     Global=&RigUnit_GetTransform(Space=LocalSpace, Item=@n1.Item).Transform
     Parent=&RigUnit_GetTransform(Space=LocalSpace, bInitial=true, Item=@n1.Item).Transform
    ).Local) {is=struct RigVMUnitNode, template="Cast::Execute(in Value,out Result)"}
   > RigVMFunctionReturnNode(OutTransform=@RigVMFunction_MathTransformToEulerTransform.Result)
  ]}
 New Function_ContainedGraph {nodes [
   RigVMFunctionEntryNode()
   > RigVMFunction_Sequence_1 = RigVMAggregateNode()
   &RigUnit_ParentConstraint(
    AdvancedSettings=(RotationOrderForFilter=XZY)
    Child=(Type=Bone,Name="base_skirt_01_l")
    Parents=((Item=(Type=Bone,Name="thigh_twist_01_l"),Weight=0.7),(Item=(Type=Bone,Name="pelvis"),Weight=0.3))
    Weight=1
    ExecutePin=@RigVMFunction_Sequence_1.B
   )
   > &RigUnit_ModifyTransforms(
    ItemToModify=((Item=(Type=Bone,Name="base_skirt_01_l"),Transform=(Rotation=(X=-0.0024,Y=-0.044,Z=-0.1637,W=0.9855),Translation=(X=0.5125,Y=7.1911,Z=1.7732),Scale3D=(X=1,Y=1,Z=1))))
    Weight=RigVMDispatchNode(BoneName=root, Default=1, Name=deformRig_l_thigh_up_alpha) {template="DISPATCH_RigDispatch_GetAnimAttribute(in Name,in BoneName,in Default,out Value,out Found)"}.Value
   )
   n1 = RigVMVariableNode() {variable=Thigh_L_SideRotation}
   > &RigUnit_ParentConstraint(
    AdvancedSettings=(RotationOrderForFilter=XZY)
    Child=(Type=Bone,Name="base_skirt_01_l")
    Filter=(TranslationFilter=(bX=false,bY=false,bZ=false),RotationFilter=(bX=True,bY=True,bZ=True),ScaleFilter=(bX=True,bY=True,bZ=True))
    Parents=((Item=(Type=Bone,Name="thigh_twist_01_l"),Weight=0.8),(Item=(Type=Bone,Name="pelvis"),Weight=0.2))
    Weight=@n1.Value
   )
   > &RigUnit_ModifyTransforms(
    ItemToModify="((Item=(Type=Bone,Name=\"base_skirt_01_l\"),Transform=(Rotation=(X=0,Y=-0.0811,Z=0,W=0.9967),Translation=(X=0.1803,Y=-2.7844,Z=5.855),Scale3D=(X=1,Y=1,Z=1))),(Item=(Type=Bone,Name=\"base_skirt_01_bck_l\"),Transform=(Rotation=(X=0.0787,Y=-0.2497,Z=0.2151,W=0.94 ...(truncated)"
    Weight=@n1.Value
   )
   &RigUnit_ModifyTransforms(
    ItemToModify="((Item=(Type=Bone,Name=\"base_skirt_01_front_r\"),Transform=(Rotation=(X=0.0684,Y=-0.0969,Z=-0.0474,W=0.9918),Translation=(X=0,Y=0,Z=2.6287),Scale3D=(X=1,Y=1,Z=1))),(Item=(Type=Bone,Name=\"base_skirt_01_front_l\"),Transform=(Rotation=(X=0.0268,Y=0.0132,Z=-0.09 ...(truncated)"
    WeightMinimum=0
    ExecutePin=@RigVMFunction_Sequence_1.G
    Weight=RigVMVariableNode() {variable=SpineRotRight}.Value
   )
   n2 = &RigVMFunction_AlphaInterp(
    InterpSpeedDecreasing=10
    InterpSpeedIncreasing=10
    Scale=1
    bClampResult=true
    bMapRange=true
    Value=&RigVMFunction_MathBoolToFloat(Value=RigVMVariableNode() {variable=KneeSliding}.Value).Result
   )
   > &RigUnit_ModifyTransforms(
    ItemToModify="((Item=(Type=Bone,Name=\"base_skirt_01_bck\"),Transform=(Rotation=(X=0,Y=0.7725,Z=0,W=-0.635),Translation=(X=0,Y=0,Z=0),Scale3D=(X=1,Y=1,Z=1))),(Item=(Type=Bone,Name=\"base_skirt_01_bck_l\"),Transform=(Rotation=(X=0,Y=0.7735,Z=0,W=-0.6338),Translation=(X=0,Y=0 ...(truncated)"
    WeightMinimum=0
    Weight=@n2.Result
   )
   n3 = &RigVMFunction_AlphaInterp(
    InterpSpeedDecreasing=10
    InterpSpeedIncreasing=10
    Scale=1
    bClampResult=true
    bMapRange=true
    Value=&RigVMFunction_MathBoolToFloat(Value=RigVMVariableNode() {variable=isDBNO}.Value).Result
   )
   > n4 = &RigUnit_ModifyTransforms(
    ItemToModify="((Item=(Type=Bone,Name=\"base_skirt_01_front_r\"),Transform=(Rotation=(X=0,Y=-0.2164,Z=0,W=0.9763),Translation=(X=5,Y=0,Z=0),Scale3D=(X=1,Y=1,Z=1))),(Item=(Type=Bone,Name=\"base_skirt_01_front_l\"),Transform=(Rotation=(X=0,Y=-0.1736,Z=0,W=0.9848),Translation=( ...(truncated)"
    WeightMinimum=0
    Weight=@n3.Result
   )
   RigVMFunction_ControlFlowBranch = &RigVMFunction_ControlFlowBranch(ExecuteContext=@n4.ExecutePin, Condition=RigVMVariableNode() {variable=IsEmoting}.Value)
   Delta Rotation Y = RigVMFunctionReferenceNode(Child=thigh_r, Parent=thigh_r_null, Source Maximum=15, ExecuteContext=@RigVMFunction_ControlFlowBranch.Completed)
   > &RigUnit_ModifyTransforms(
    ItemToModify="((Item=(Type=Bone,Name=\"base_skirt_01_front_r\"),Transform=(Rotation=(X=-0.0227,Y=-0.1285,Z=-0.1722,W=0.9764),Translation=(X=0,Y=0,Z=0),Scale3D=(X=1,Y=1,Z=1))),(Item=(Type=Bone,Name=\"base_skirt_01_front_l\"),Transform=(Rotation=(X=-0.0227,Y=-0.1285,Z=-0.1722 ...(truncated)"
    WeightMinimum=0
    Weight=@"Delta Rotation Y".Delta Rotation
   )
   n5 = &RigUnit_SphericalPoseReader(
    ActiveRegionScaleFactors=(PositiveWidth=1,NegativeWidth=1,PositiveHeight=1,NegativeHeight=1)
    ActiveRegionSize=0.05
    Debug=(bDrawDebug=false,bDraw2D=False,bDrawLocalAxes=False,DebugScale=25,DebugSegments=20,DebugThickness=0.25)
    DriverAxis=(X=-1,Y=0,Z=0)
    DriverItem=(Type=Bone,Name="thigh_l")
    FalloffRegionScaleFactors=(PositiveWidth=1,NegativeWidth=1,PositiveHeight=1,NegativeHeight=1)
    FalloffSize=0.55
    FlipHeightScaling=False
    FlipWidthScaling=False
    OptionalParentItem=(Type=Bone,Name="None")
    RotationOffset=(X=140,Y=30,Z=0)
    ExecutePin=@RigVMFunction_ControlFlowBranch.True
   )
   > &RigUnit_ModifyTransforms(
    ItemToModify=((Item=(Type=Bone,Name="base_skirt_01_l"),Transform=(Rotation=(X=0,Y=-0.0666,Z=0,W=0.9978),Translation=(X=-2.4474,Y=0,Z=2.4628),Scale3D=(X=1,Y=1,Z=1))))
    WeightMinimum=0
    Weight=@n5.OutputParam
   )
   RigVMFunctionReturnNode(ExecuteContext=@RigVMFunction_Sequence_1.H)
   n6 = RigVMDispatchNode(BoneName=root, Default=1, Name=deformRig_r_thigh_up_alpha) {template="DISPATCH_RigDispatch_GetAnimAttribute(in Name,in BoneName,in Default,out Value,out Found)"}
   Remap = &RigVMFunction_MathDoubleRemap(SourceMaximum=1, SourceMinimum=0, TargetMaximum=1, TargetMinimum=0.4, bClamp=False, Value=@n6.Value)
   &RigUnit_ParentConstraint(
    AdvancedSettings=(RotationOrderForFilter=XZY)
    Child=(Type=Bone,Name="base_skirt_01_r")
    Parents=((Item=(Type=Bone,Name="thigh_twist_01_r"),Weight=0.7),(Item=(Type=Bone,Name="pelvis"),Weight=0.3))
    Weight=@Remap.Result
    ExecutePin=@RigVMFunction_Sequence_1.C
   )
   > &RigUnit_ModifyTransforms(
    ItemToModify=((Item=(Type=Bone,Name="base_skirt_01_r"),Transform=(Rotation=(X=-0.0408,Y=0.0878,Z=-0.4192,W=-0.9027),Translation=(X=-2.6454,Y=-7.3094,Z=2.5332),Scale3D=(X=1,Y=1,Z=1))))
    Weight=@n6.Value
   )
   > &RigUnit_ModifyTransforms(
    ItemToModify=((Item=(Type=Bone,Name="base_skirt_01_r"),Transform=(Rotation=(X=0,Y=0,Z=0.089,W=0.996),Translation=(X=-4.0368,Y=-0.4325,Z=0.3321),Scale3D=(X=1,Y=1,Z=1))))
    WeightMinimum=0
    Weight=RigVMVariableNode() {variable=Thigh_FrontSkirt_r}.Value
   )
   n7 = &RigVMFunction_AlphaInterp(
    InterpSpeedDecreasing=10
    InterpSpeedIncreasing=10
    OutRange=(Min=1,Max=0)
    Scale=1
    bClampResult=true
    bMapRange=true
    Value=&RigVMFunction_MathBoolToFloat(Value=RigVMVariableNode() {variable=isCrouching}.Value).Result
   )
   > &RigUnit_ModifyTransforms(
    ItemToModify=((Item=(Type=Bone,Name="base_skirt_01_r"),Transform=(Rotation=(X=-0.4859,Y=0,Z=0,W=0.874),Translation=(X=0,Y=3.1991,Z=1.9875),Scale3D=(X=1,Y=1,Z=1))))
    WeightMinimum=0
    Weight=&RigVMFunction_MathDoubleMul(B=@n7.Result, A=RigVMVariableNode() {variable=Thigh_FrontSkirt_l}.Value).Result
   )
   n8 = RigVMVariableNode() {variable=Thigh_R_SideRotation}
   > &RigUnit_ParentConstraint(
    AdvancedSettings=(RotationOrderForFilter=XZY)
    Child=(Type=Bone,Name="base_skirt_01_r")
    Filter=(TranslationFilter=(bX=false,bY=false,bZ=false),RotationFilter=(bX=True,bY=True,bZ=True),ScaleFilter=(bX=True,bY=True,bZ=True))
    Parents=((Item=(Type=Bone,Name="thigh_twist_01_r"),Weight=0.8),(Item=(Type=Bone,Name="pelvis"),Weight=0.2))
    Weight=&RigVMFunction_MathDoubleRemap(SourceMaximum=1, SourceMinimum=0, TargetMaximum=1, TargetMinimum=0.4, bClamp=False, Value=@n8.Value).Result
   )
   > &RigUnit_ModifyTransforms(
    ItemToModify="((Item=(Type=Bone,Name=\"base_skirt_01_r\"),Transform=(Rotation=(X=-0.0072,Y=-0.0981,Z=0.1098,W=0.9891),Translation=(X=-3.4526,Y=-0.7695,Z=9.1573),Scale3D=(X=1,Y=1,Z=1))),(Item=(Type=Bone,Name=\"base_skirt_01_bck_r\"),Transform=(Rotation=(X=-0.2839,Y=-0.1901,Z ...(truncated)"
    Weight=@n8.Value
   )
   &RigUnit_ModifyTransforms(
    ItemToModify="((Item=(Type=Bone,Name=\"dyn_skirt_01_bck\"),Transform=(Rotation=(X=0,Y=-0.1459,Z=0,W=0.9893),Translation=(X=0,Y=0,Z=0),Scale3D=(X=1,Y=1,Z=1))),(Item=(Type=Bone,Name=\"dyn_skirt_01_bck_l\"),Transform=(Rotation=(X=0,Y=-0.0659,Z=0,W=0.9978),Translation=(X=0,Y=0, ...(truncated)"
    WeightMinimum=0
    Weight=RigVMVariableNode() {variable=PelvisToGround}.Value
    ExecutePin=@RigVMFunction_Sequence_1.E
   )
   > RigVMFunction_Sequence_2 = RigVMAggregateNode()
   n9 = &RigUnit_ModifyTransforms(
    ItemToModify="((Item=(Type=Bone,Name=\"dyn_skirt_01_bck\"),Transform=(Rotation=(X=-0.1798,Y=0.2951,Z=-0.0107,W=-0.9383),Translation=(X=0,Y=0,Z=0),Scale3D=(X=1,Y=1,Z=1))),(Item=(Type=Bone,Name=\"dyn_skirt_01_bck_l\"),Transform=(Rotation=(X=-0.134,Y=0.395,Z=0.1406,W=-0.8979), ...(truncated)"
    WeightMinimum=0
    ExecutePin=@RigVMFunction_Sequence_2.A
    Weight=RigVMVariableNode() {variable=Thigh_L_BackRotation}.Value
   )
   n10 = &RigVMFunction_ControlFlowBranch(ExecuteContext=@n9.ExecutePin, Condition=RigVMVariableNode() {variable=isCrouching}.Value)
   Multiply = &RigVMFunction_MathDoubleMul(
    B=&RigVMFunction_MathDoubleRemap(
     SourceMaximum=1
     SourceMinimum=0
     TargetMaximum=0
     TargetMinimum=1
     Value=RigVMDispatchNode(BoneName=root, Default=1, Name=deformRig_l_thigh_up_alpha) {template="DISPATCH_RigDispatch_GetAnimAttribute(in Name,in BoneName,in Default,out Value,out Found)"}.Value
    ).Result
    A=RigVMVariableNode() {variable=Calf_L_Rotation}.Value
   )
   &RigUnit_ModifyTransforms(
    ItemToModify="((Item=(Type=Bone,Name=\"dyn_skirt_02_bck\"),Transform=(Rotation=(X=0.3314,Y=-0.6583,Z=-0.0384,W=0.6748),Translation=(X=0,Y=0,Z=0),Scale3D=(X=1,Y=1,Z=1))),(Item=(Type=Bone,Name=\"dyn_skirt_02_bck_l\"),Transform=(Rotation=(X=-0.0425,Y=-0.9278,Z=-0.1914,W=0.3175 ...(truncated)"
    WeightMinimum=0
    ExecutePin=@n10.False
    Weight=@Multiply.Result
   )
   &RigUnit_ModifyTransforms(
    ItemToModify="((Item=(Type=Bone,Name=\"dyn_skirt_01_l\"),Transform=(Rotation=(X=-0.0069,Y=-0.0485,Z=-0.1405,W=0.9889),Translation=(X=0,Y=0,Z=0),Scale3D=(X=1,Y=1,Z=1))),(Item=(Type=Bone,Name=\"dyn_skirt_01_r\"),Transform=(Rotation=(X=0,Y=0,Z=0,W=1),Translation=(X=0,Y=0,Z=0), ...(truncated)"
    WeightMinimum=0
    ExecutePin=@n10.True
    Weight=@Multiply.Result
   )
   &RigUnit_ModifyTransforms(
    ItemToModify="((Item=(Type=Bone,Name=\"dyn_skirt_01_bck\"),Transform=(Rotation=(X=0.1926,Y=0.3004,Z=-0.0014,W=-0.9342),Translation=(X=0,Y=0,Z=0),Scale3D=(X=1,Y=1,Z=1))),(Item=(Type=Bone,Name=\"dyn_skirt_01_bck_l\"),Transform=(Rotation=(X=0,Y=0,Z=0,W=1),Translation=(X=0,Y=0, ...(truncated)"
    WeightMinimum=0
    Weight=RigVMVariableNode() {variable=Thigh_R_BackRotation}.Value
    ExecutePin=@RigVMFunction_Sequence_2.B
   )
   n11 = RigVMDispatchNode(BoneName=root, Default=1, Name=deformRig_r_thigh_up_alpha) {template="DISPATCH_RigDispatch_GetAnimAttribute(in Name,in BoneName,in Default,out Value,out Found)"}
   > &RigUnit_ModifyTransforms(
    ItemToModify="((Item=(Type=Bone,Name=\"dyn_skirt_02_bck\"),Transform=(Rotation=(X=-0.0812,Y=-0.6216,Z=0.0116,W=0.779),Translation=(X=0,Y=0,Z=0),Scale3D=(X=1,Y=1,Z=1))),(Item=(Type=Bone,Name=\"dyn_skirt_02_bck_l\"),Transform=(Rotation=(X=-0.0006,Y=-0.0244,Z=-0.0262,W=0.9994) ...(truncated)"
    WeightMinimum=0
    Weight=&RigVMFunction_MathDoubleMul(
     B=&RigVMFunction_MathDoubleRemap(SourceMaximum=1, SourceMinimum=0, TargetMaximum=0, TargetMinimum=1, Value=@n11.Value).Result
     A=RigVMVariableNode() {variable=Calf_R_Rotation}.Value
    ).Result
   )
   > &RigUnit_ModifyTransforms(
    ItemToModify="((Item=(Type=Bone,Name=\"dyn_skirt_02_bck\"),Transform=(Rotation=(X=0,Y=0.2222,Z=0,W=0.975),Translation=(X=0,Y=0,Z=0),Scale3D=(X=1,Y=1,Z=1))),(Item=(Type=Bone,Name=\"dyn_skirt_02_bck_l\"),Transform=(Rotation=(X=0,Y=0.2691,Z=0,W=0.9631),Translation=(X=0,Y=0,Z=0 ...(truncated)"
    WeightMinimum=0
    Weight=RigVMVariableNode() {variable=inFloating}.Value
   )
   &RigUnit_ModifyTransforms(
    ItemToModify="((Item=(Type=Bone,Name=\"dyn_skirt_01_front_l\"),Transform=(Rotation=(X=-0.0002,Y=-0.0886,Z=-0.0061,W=0.996),Translation=(X=0,Y=0,Z=0),Scale3D=(X=1,Y=1,Z=1))),(Item=(Type=Bone,Name=\"dyn_skirt_01_front_r\"),Transform=(Rotation=(X=0,Y=-0.0262,Z=0,W=0.9997),Tran ...(truncated)"
    WeightMinimum=0
    Weight=RigVMVariableNode() {variable=SpineRotRight}.Value
    ExecutePin=@RigVMFunction_Sequence_2.C
   )
   > &RigUnit_ModifyTransforms(
    ItemToModify="((Item=(Type=Bone,Name=\"dyn_skirt_03_bck_l\"),Transform=(Rotation=(X=0.3051,Y=0.5474,Z=0.1771,W=-0.7589),Translation=(X=-1.8854,Y=8.8745,Z=6.3442),Scale3D=(X=1,Y=1,Z=1))),(Item=(Type=Bone,Name=\"dyn_skirt_03_bck\"),Transform=(Rotation=(X=0,Y=-0.4245,Z=0,W=0.9 ...(truncated)"
    WeightMinimum=0
    Weight=&RigVMFunction_MathDoubleMul(B=RigVMVariableNode() {variable=skydiveThigBackL}.Value, A=RigVMVariableNode() {variable=isSkydiving}.Value).Result
   )
   AlphaInterp = &RigVMFunction_AlphaInterp(
    InterpSpeedDecreasing=10
    InterpSpeedIncreasing=10
    OutRange=(Min=1,Max=0)
    Scale=1
    bClampResult=true
    bMapRange=true
    Value=&RigVMFunction_MathBoolToFloat(Value=RigVMVariableNode() {variable=glider}.Value).Result
   )
   Multiply_2_1 = RigVMAggregateNode(C=@AlphaInterp.Result, A=RigVMVariableNode() {variable=isSkydiving}.Value, B=RigVMVariableNode() {variable=skydiveThigBackR}.Value)
   > &RigUnit_ModifyTransforms(
    ItemToModify="((Item=(Type=Bone,Name=\"dyn_skirt_03_r\"),Transform=(Rotation=(X=0,Y=-0.0561,Z=0,W=0.9984),Translation=(X=0,Y=0,Z=0),Scale3D=(X=1,Y=1,Z=1))),(Item=(Type=Bone,Name=\"dyn_skirt_03_bck_r\"),Transform=(Rotation=(X=-0.1725,Y=0.4775,Z=0.3233,W=-0.7986),Translation= ...(truncated)"
    WeightMinimum=0
    Weight=@Multiply_2_1.Result
   )
   Multiply_3 = RigVMAggregateNode(
    C=RigVMVariableNode() {variable=openParachute}.Value
    B=RigVMVariableNode() {variable=spineRotBack}.Value
    A=RigVMVariableNode() {variable=isSkydiving}.Value
   )
   > &RigUnit_ModifyTransforms(
    ItemToModify="((Item=(Type=Bone,Name=\"dyn_skirt_02_bck_l\"),Transform=(Rotation=(X=0,Y=0.3866,Z=0,W=0.9222),Translation=(X=0,Y=0,Z=0),Scale3D=(X=1,Y=1,Z=1))),(Item=(Type=Bone,Name=\"dyn_skirt_02_bck\"),Transform=(Rotation=(X=0,Y=0.4846,Z=0,W=0.8747),Translation=(X=0,Y=0,Z= ...(truncated)"
    WeightMinimum=0
    Weight=@Multiply_3.Result
   )
   RigVMFunction_MathBoolToFloat = &RigVMFunction_MathBoolToFloat(Value=RigVMVariableNode() {variable=IsInAir}.Value)
   > n12 = &RigUnit_ModifyTransforms(
    ItemToModify=((Item=(Type=Bone,Name="dyn_skirt_01_l"),Transform=(Rotation=(X=0,Y=-0.0619,Z=0,W=0.9981),Translation=(X=0,Y=0,Z=0),Scale3D=(X=1,Y=1,Z=1))))
    WeightMinimum=0
    Weight=@RigVMFunction_MathBoolToFloat.Result
   )
   n13 = &RigVMFunction_ControlFlowBranch(ExecuteContext=@n12.ExecutePin, Condition=RigVMVariableNode() {variable=TacticalSprinting}.Value)
   Delta Rotation Z_1_1_1 = RigVMFunctionReferenceNode(Child=calf_r, Parent=calf_r_null, Source Maximum=110, Source Minimum=50, ExecuteContext=@n13.True)
   n14 = RigVMDispatchNode(BoneName=root, Default=1, Name=deformRig_r_thigh_up_alpha) {template="DISPATCH_RigDispatch_GetAnimAttribute(in Name,in BoneName,in Default,out Value,out Found)"}
   > &RigUnit_ModifyTransforms(
    ItemToModify="((Item=(Type=Bone,Name=\"dyn_skirt_02_bck_r\"),Transform=(Rotation=(X=-0.0892,Y=-0.0539,Z=0.0048,W=-0.9945),Translation=(X=-0.3864,Y=0,Z=2.6305),Scale3D=(X=1,Y=1,Z=1))),(Item=(Type=Bone,Name=\"dyn_skirt_02_bck\"),Transform=(Rotation=(X=-0.3151,Y=-0.1571,Z=-0.0 ...(truncated)"
    WeightMinimum=0
    Weight=&RigVMFunction_MathDoubleMul(
     A=@"Delta Rotation Z_1_1_1".Delta Rotation
     B=&RigVMFunction_MathDoubleRemap(SourceMaximum=1, SourceMinimum=0, TargetMaximum=0, TargetMinimum=1, Value=@n14.Value).Result
    ).Result
   )
   > Delta Rotation Z_1_1 = RigVMFunctionReferenceNode(Child=calf_l, Parent=calf_l_null, Source Maximum=110, Source Minimum=50)
   n15 = RigVMDispatchNode(BoneName=root, Default=1, Name=deformRig_l_thigh_up_alpha) {template="DISPATCH_RigDispatch_GetAnimAttribute(in Name,in BoneName,in Default,out Value,out Found)"}
   > &RigUnit_ModifyTransforms(
    ItemToModify="((Item=(Type=Bone,Name=\"dyn_skirt_02_bck_l\"),Transform=(Rotation=(X=0,Y=-0.0841,Z=0,W=-0.9965),Translation=(X=-0.3864,Y=0,Z=2.6305),Scale3D=(X=1,Y=1,Z=1))),(Item=(Type=Bone,Name=\"dyn_skirt_02_bck\"),Transform=(Rotation=(X=0,Y=0.1463,Z=0,W=-0.9892),Translati ...(truncated)"
    WeightMinimum=0
    Weight=&RigVMFunction_MathDoubleMul(
     A=@"Delta Rotation Z_1_1".Delta Rotation
     B=&RigVMFunction_MathDoubleRemap(SourceMaximum=1, SourceMinimum=0, TargetMaximum=0, TargetMinimum=1, Value=@n15.Value).Result
    ).Result
   )
   Add = &RigVMFunction_MathDoubleAdd(
    B=RigVMDispatchNode(BoneName=root, Default=1, Name=deformRig_r_thigh_up_alpha) {template="DISPATCH_RigDispatch_GetAnimAttribute(in Name,in BoneName,in Default,out Value,out Found)"}.Value
    A=RigVMDispatchNode(BoneName=root, Default=1, Name=deformRig_l_thigh_up_alpha) {template="DISPATCH_RigDispatch_GetAnimAttribute(in Name,in BoneName,in Default,out Value,out Found)"}.Value
   )
   &RigUnit_ModifyTransforms(
    ItemToModify=((Item=(Type=Bone,Name="dyn_belt"),Transform=(Rotation=(X=0,Y=0,Z=0,W=1),Translation=(X=-14.8869,Y=0,Z=0),Scale3D=(X=1,Y=1,Z=1))))
    WeightMinimum=0
    Weight=&RigVMFunction_MathDoubleRemap(SourceMaximum=1, SourceMinimum=0, TargetMaximum=1, TargetMinimum=0, Value=@Add.Result).Result
    ExecutePin=@RigVMFunction_Sequence_1.A
   )
   n16 = &RigVMFunction_AlphaInterp(
    InterpSpeedDecreasing=10
    InterpSpeedIncreasing=10
    OutRange=(Min=1,Max=0)
    Scale=1
    bClampResult=true
    bMapRange=true
    Value=&RigVMFunction_MathBoolToFloat(Value=RigVMVariableNode() {variable=isCrouching}.Value).Result
   )
   > &RigUnit_ModifyTransforms(
    ItemToModify="((Item=(Type=Bone,Name=\"base_skirt_01_front_l\"),Transform=(Rotation=(X=0,Y=0,Z=0,W=1),Translation=(X=0,Y=1.5,Z=0),Scale3D=(X=1,Y=1,Z=1))),(Item=(Type=Bone,Name=\"base_skirt_01_front_r\"),Transform=(Rotation=(X=0,Y=0,Z=0,W=1),Translation=(X=0,Y=-1.5,Z=0),Scal ...(truncated)"
    WeightMinimum=0
    Weight=@n16.Result
   )
   > n17 = &RigUnit_SphericalPoseReader(
    ActiveRegionScaleFactors=(PositiveWidth=1,NegativeWidth=0,PositiveHeight=1,NegativeHeight=1)
    ActiveRegionSize=0.3
    Debug=(bDrawDebug=false,bDraw2D=false,bDrawLocalAxes=false,DebugScale=30,DebugSegments=30,DebugThickness=0.2)
    DriverAxis=(X=-1,Y=0,Z=0)
    DriverItem=(Type=Bone,Name="thigh_l")
    FalloffRegionScaleFactors=(PositiveWidth=1,NegativeWidth=1.2,PositiveHeight=1,NegativeHeight=1)
    FalloffSize=0.5
    OptionalParentItem=(Type=Bone,Name="pelvis")
    RotationOffset=(X=90,Y=0,Z=-15)
   )
   > RigVMVariableNode(Value=@n17.OutputParam) {variable=Thigh_FrontSkirt_l}
   > n18 = &RigUnit_SphericalPoseReader(
    ActiveRegionScaleFactors=(PositiveWidth=1,NegativeWidth=0,PositiveHeight=1,NegativeHeight=1)
    ActiveRegionSize=0.3
    Debug=(bDrawDebug=false,bDraw2D=false,bDrawLocalAxes=false,DebugScale=30,DebugSegments=30,DebugThickness=0.2)
    DriverAxis=(X=1,Y=0,Z=0)
    DriverItem=(Type=Bone,Name="thigh_r")
    FalloffRegionScaleFactors=(PositiveWidth=1,NegativeWidth=1.2,PositiveHeight=1,NegativeHeight=1)
    FalloffSize=0.5
    OptionalParentItem=(Type=Bone,Name="pelvis")
    RotationOffset=(X=-90,Y=180,Z=-15)
   )
   > RigVMVariableNode(Value=@n18.OutputParam) {variable=Thigh_FrontSkirt_r}
   n19 = &RigVMFunction_AlphaInterp(
    InterpSpeedDecreasing=4
    InterpSpeedIncreasing=10
    Scale=1
    bClampResult=true
    bMapRange=true
    Value=RigVMDispatchNode(
     Condition=/Script/RigVM.RigVMFunction_MathDoubleGreater(A=@n18.OutputParam, B=@n17.OutputParam) {is=struct RigVMUnitNode, template="Greater::Execute(in A,in B,out Result)"}.Result
     True=@n18.OutputParam
     False=@n17.OutputParam
    ) {template="DISPATCH_RigVMDispatch_If(in Condition,in True,in False,out Result)"}.Result
   )
   > &RigUnit_ModifyTransforms(
    ItemToModify=((Item=(Type=Bone,Name="base_skirt_01_front_l"),Transform=(Rotation=(X=0.0168,Y=-0.7723,Z=-0.0108,W=0.635),Translation=(X=-13.9944,Y=3.6857,Z=0.2388),Scale3D=(X=1,Y=1,Z=1))))
    WeightMinimum=0
    Weight=@n19.Result
   )
   > &RigUnit_ModifyTransforms(
    ItemToModify=((Item=(Type=Bone,Name="base_skirt_01_front_r"),Transform=(Rotation=(X=-0.0201,Y=-0.7715,Z=0.0138,W=0.6357),Translation=(X=-14.4464,Y=-4.1936,Z=0.2088),Scale3D=(X=1,Y=1,Z=1))))
    WeightMinimum=0
    Weight=@n19.Result
   )
   &RigVMFunction_MathQuaternionFromEuler(Euler=(X=-15,Y=-35,Z=10))
   &RigVMFunction_MathQuaternionFromEuler(Euler=(X=-20,Y=48,Z=-32))
   &RigVMFunction_MathQuaternionFromEuler(Euler=(X=0,Y=20,Z=0))
   &RigVMFunction_MathQuaternionFromEuler(Euler=(X=0,Y=5,Z=5))
   &RigVMFunction_MathQuaternionFromEuler(Euler=(X=0,Y=5,Z=15))
  ]}
 New Function_ContainedGraph {nodes [
   RigVMFunctionEntryNode()
   > RigVMFunction_Sequence = RigVMAggregateNode()
   Delta Rotation Y = RigVMFunctionReferenceNode(Child=thigh_l, Parent=thigh_l_null, Source Maximum=-75, Source Minimum=0, ExecuteContext=@RigVMFunction_Sequence.A)
   > RigVMVariableNode(Value=@"Delta Rotation Y".Delta Rotation) {variable=Thigh_L_SideRotation}
   > Delta Rotation Y_1 = RigVMFunctionReferenceNode(Child=thigh_r, Parent=thigh_r_null, Source Maximum=-75, Source Minimum=0)
   > RigVMVariableNode(Value=@"Delta Rotation Y_1".Delta Rotation) {variable=Thigh_R_SideRotation}
   > Delta Rotation Y_2 = RigVMFunctionReferenceNode(Child=thigh_l, Parent=thigh_l_null, Source Maximum=15, Source Minimum=0)
   > RigVMVariableNode(Value=@"Delta Rotation Y_2".Delta Rotation) {variable=Thigh_L_In_SideRotation}
   > RigVMVariableNode(Value=&RigVMFunction_MathDoubleRemap(
     SourceMaximum=50
     SourceMinimum=70
     TargetMaximum=1
     TargetMinimum=0
     Value=&RigUnit_GetRelativeTransformForItem(Child=(Type=Bone,Name="spine_01"), Parent=(Type=Bone,Name="root"), bChildInitial=False, bParentInitial=False).RelativeTransform.Translation.Z
    ).Result) {variable=PelvisToGround}
   > Delta Rotation Z = RigVMFunctionReferenceNode(Child=thigh_l, Parent=thigh_l_null, Source Maximum=50, Source Minimum=5)
   > RigVMVariableNode(Value=@"Delta Rotation Z".Delta Rotation) {variable=Thigh_L_BackRotation}
   > Delta Rotation Z_1 = RigVMFunctionReferenceNode(Child=thigh_r, Parent=thigh_r_null, Source Maximum=50, Source Minimum=5)
   > RigVMVariableNode(Value=@"Delta Rotation Z_1".Delta Rotation) {variable=Thigh_R_BackRotation}
   > Delta Rotation Z_1_1 = RigVMFunctionReferenceNode(Child=calf_l, Parent=calf_l_null, Source Maximum=130, Source Minimum=15)
   > RigVMVariableNode(Value=@"Delta Rotation Z_1_1".Delta Rotation) {variable=Calf_L_Rotation}
   > Delta Rotation Z_1_1_1 = RigVMFunctionReferenceNode(Child=calf_r, Parent=calf_r_null, Source Maximum=130, Source Minimum=15)
   > RigVMVariableNode(Value=@"Delta Rotation Z_1_1_1".Delta Rotation) {variable=Calf_R_Rotation}
   n1 = RigVMVariableNode() {variable=IsInAir}
   > RigVMVariableNode(Value=&RigVMFunction_MathDoubleRemap(
     SourceMaximum=1
     SourceMinimum=0
     TargetMaximum=1
     TargetMinimum=0
     Value=RigVMDispatchNode(False=0, True=1, Condition=@n1.Value) {template="DISPATCH_RigVMDispatch_If(in Condition,in True,in False,out Result)"}.Result
    ).Result) {variable=isSkydiving}
   > RigVMVariableNode(Value=&RigVMFunction_MathDoubleRemap(SourceMaximum=1, SourceMinimum=0, TargetMaximum=0, TargetMinimum=1, Value=RigVMVariableNode() {variable=isSkydiving}.Value).Result) {variable=NotInAir}
   > Delta Rotation Z_2 = RigVMFunctionReferenceNode(Child=head, Parent=neck_01_null, Source Maximum=-63, Source Minimum=-21)
   > RigVMVariableNode(Value=@"Delta Rotation Z_2".Delta Rotation) {variable=headLookUp}
   Delta Rotation X = RigVMFunctionReferenceNode(Child=pelvis, Parent=spine_05_null, Source Maximum=-38, Source Minimum=-15, ExecuteContext=@RigVMFunction_Sequence.B)
   > RigVMVariableNode(Value=@"Delta Rotation X".Delta Rotation) {variable=SpineRotRight}
   > Delta Rotation Z_3 = RigVMFunctionReferenceNode(Child=thigh_l, Parent=thigh_l_null, Source Maximum=22, Source Minimum=0)
   > RigVMVariableNode(Value=@"Delta Rotation Z_3".Delta Rotation) {variable=skydiveThigBackL}
   > Delta Rotation Z_3_1 = RigVMFunctionReferenceNode(Child=thigh_r, Parent=thigh_r_null, Source Maximum=22, Source Minimum=0)
   > RigVMVariableNode(Value=@"Delta Rotation Z_3_1".Delta Rotation) {variable=skydiveThigBackR}
   > Delta Rotation Z_3_2 = RigVMFunctionReferenceNode(Child=pelvis, Parent=spine_05_null, Source Maximum=58, Source Minimum=20)
   > RigVMVariableNode(Value=@"Delta Rotation Z_3_2".Delta Rotation) {variable=spineRotBack}
   n2 = RigVMVariableNode() {variable=IsInAir}
   Remap = &RigVMFunction_MathDoubleRemap(
    SourceMaximum=1
    SourceMinimum=0
    TargetMaximum=1
    TargetMinimum=0
    bClamp=False
    Value=RigVMDispatchNode(False=0, True=1, Condition=@n2.Value) {template="DISPATCH_RigVMDispatch_If(in Condition,in True,in False,out Result)"}.Result
   )
   Multiply = &RigVMFunction_MathDoubleMul(
    B=&RigVMFunction_MathDoubleRemap(
     SourceMaximum=0.2
     SourceMinimum=0
     TargetMaximum=1
     TargetMinimum=0
     bClamp=False
     Value=RigVMDispatchNode(BoneName=root, Default=1, Name=deformRig_l_upperarm_up_alpha) {template="DISPATCH_RigDispatch_GetAnimAttribute(in Name,in BoneName,in Default,out Value,out Found)"}.Value
    ).Result
    A=@Remap.Result
   )
   > RigVMVariableNode(Value=@Multiply.Result) {variable=ArmUpSkydiveL}
   n3 = RigVMDispatchNode(BoneName=root, Default=1, Name=deformRig_r_upperarm_up_alpha) {template="DISPATCH_RigDispatch_GetAnimAttribute(in Name,in BoneName,in Default,out Value,out Found)"}
   > RigVMVariableNode(Value=&RigVMFunction_MathDoubleMul(
     B=&RigVMFunction_MathDoubleRemap(SourceMaximum=0.2, SourceMinimum=0, TargetMaximum=1, TargetMinimum=0, bClamp=False, Value=@n3.Value).Result
     A=@Remap.Result
    ).Result) {variable=ArmUpSkydiveR}
   RigVMFunctionReturnNode(ExecuteContext=@RigVMFunction_Sequence.C)
   RigVMFunction_MathBoolOr = &RigVMFunction_MathBoolOr(A=@n2.Value, B=RigVMVariableNode() {variable=IsEmoting}.Value)
  ]}
 New Function_ContainedGraph {nodes [
   n1 = RigVMFunctionEntryNode()
   RigVMFunction_MathQuaternionToEuler = &RigVMFunction_MathQuaternionToEuler(
    RotationOrder=XYZ
    Value=&RigUnit_GetRelativeTransformForItem(
     Child=(Type=Bone,Name="hand_l")
     Parent=(Type=Null,Name="lowerarm_l")
     bChildInitial=False
     bParentInitial=False
     Child.Name=@n1.Child
     Parent.Name=@n1.Parent
    ).RelativeTransform.Rotation
   )
   Remap = &RigVMFunction_MathDoubleRemap(
    TargetMaximum=1
    TargetMinimum=0
    Value=@RigVMFunction_MathQuaternionToEuler.Result.X
    SourceMaximum=@n1.Source Maximum
    SourceMinimum=@n1.Source Minimum
   )
   > RigVMFunctionReturnNode(Delta Rotation=@Remap.Result)
  ]}
 New Function_ContainedGraph {nodes [
   n1 = RigVMFunctionEntryNode()
   RigVMFunction_Sequence = &RigVMFunction_Sequence(ExecuteContext=@n1.ExecuteContext)
   &RigUnit_ModifyTransforms(
    ItemToModify="((Item=(Type=Bone,Name=\"base_cape_out_01_r\"),Transform=(Rotation=(X=0,Y=0,Z=0.0698,W=0.9976),Translation=(X=-5.6666,Y=0,Z=2.2604),Scale3D=(X=1,Y=1,Z=1))),(Item=(Type=Bone,Name=\"base_cape_in_01_r\"),Transform=(Rotation=(X=0,Y=0,Z=0,W=1),Translation=(X=-3.398 ...(truncated)"
    WeightMinimum=0
    Weight=RigVMDispatchNode(BoneName=root, Default=1, Name=deformRig_r_upperarm_up_alpha) {template="DISPATCH_RigDispatch_GetAnimAttribute(in Name,in BoneName,in Default,out Value,out Found)"}.Value
    ExecutePin=@RigVMFunction_Sequence.A
   )
   > &RigUnit_ModifyTransforms(
    ItemToModify="((Item=(Type=Bone,Name=\"base_cape_out_01_l\"),Transform=(Rotation=(X=0,Y=0,Z=-0.0698,W=0.9976),Translation=(X=-5.6666,Y=0,Z=2.2604),Scale3D=(X=1,Y=1,Z=1))),(Item=(Type=Bone,Name=\"base_cape_in_01_l\"),Transform=(Rotation=(X=0,Y=0,Z=0,W=1),Translation=(X=-3.39 ...(truncated)"
    WeightMinimum=0
    Weight=RigVMDispatchNode(BoneName=root, Default=1, Name=deformRig_l_upperarm_up_alpha) {template="DISPATCH_RigDispatch_GetAnimAttribute(in Name,in BoneName,in Default,out Value,out Found)"}.Value
   )
   > Delta Rotation Z = RigVMFunctionReferenceNode(Child=upperarm_r, Parent=upperarm_r_null, Source Maximum=-75)
   > &RigUnit_ModifyTransforms(
    ItemToModify="((Item=(Type=Bone,Name=\"base_cape_out_01_r\"),Transform=(Rotation=(X=-0.4085,Y=0.0327,Z=0.0765,W=0.909),Translation=(X=-2,Y=4,Z=10),Scale3D=(X=1,Y=1,Z=1))),(Item=(Type=Bone,Name=\"base_cape_in_01_r\"),Transform=(Rotation=(X=-0.1117,Y=-0.0713,Z=0.0094,W=0.9911 ...(truncated)"
    WeightMinimum=0
    Weight=@"Delta Rotation Z".Delta Rotation
   )
   > Delta Rotation Z_1 = RigVMFunctionReferenceNode(Child=upperarm_l, Parent=upperarm_l_null, Source Maximum=-75)
   > &RigUnit_ModifyTransforms(
    ItemToModify="((Item=(Type=Bone,Name=\"base_cape_out_01_l\"),Transform=(Rotation=(X=0.4085,Y=0.0327,Z=-0.0765,W=0.909),Translation=(X=-2,Y=-4,Z=10),Scale3D=(X=1,Y=1,Z=1))),(Item=(Type=Bone,Name=\"base_cape_in_01_l\"),Transform=(Rotation=(X=0.1117,Y=-0.0713,Z=-0.0094,W=0.991 ...(truncated)"
    WeightMinimum=0
    Weight=@"Delta Rotation Z_1".Delta Rotation
   )
   > &RigUnit_ModifyTransforms(
    ItemToModify="((Item=(Type=Bone,Name=\"base_cape_out_01_r\"),Transform=(Rotation=(X=-0.0137,Y=-0.1409,Z=0.0826,W=0.9865),Translation=(X=-2.4197,Y=-0.5906,Z=0.2695),Scale3D=(X=1,Y=1,Z=1))),(Item=(Type=Bone,Name=\"base_cape_in_01_r\"),Transform=(Rotation=(X=0,Y=-0.0459,Z=0,W= ...(truncated)"
    WeightMinimum=0
    Weight=RigVMDispatchNode(BoneName=root, Default=1, Name=deformRig_r_upperarm_fr_alpha) {template="DISPATCH_RigDispatch_GetAnimAttribute(in Name,in BoneName,in Default,out Value,out Found)"}.Value
   )
   > &RigUnit_ModifyTransforms(
    ItemToModify="((Item=(Type=Bone,Name=\"base_cape_out_01_l\"),Transform=(Rotation=(X=0.0137,Y=-0.1409,Z=-0.0826,W=0.9865),Translation=(X=-2.4197,Y=-0.5906,Z=-0.2695),Scale3D=(X=1,Y=1,Z=1))),(Item=(Type=Bone,Name=\"base_cape_in_01_l\"),Transform=(Rotation=(X=0,Y=-0.0459,Z=0,W ...(truncated)"
    WeightMinimum=0
    Weight=RigVMDispatchNode(BoneName=root, Default=1, Name=deformRig_l_upperarm_fr_alpha) {template="DISPATCH_RigDispatch_GetAnimAttribute(in Name,in BoneName,in Default,out Value,out Found)"}.Value
   )
   > Delta Rotation Y_2 = RigVMFunctionReferenceNode(Child=upperarm_r, Parent=upperarm_r_null, Source Maximum=45)
   > &RigUnit_ModifyTransforms(
    ItemToModify="((Item=(Type=Bone,Name=\"base_cape_out_01_r\"),Transform=(Rotation=(X=0,Y=0,Z=0.0175,W=0.9998),Translation=(X=0,Y=3,Z=0),Scale3D=(X=1,Y=1,Z=1))),(Item=(Type=Bone,Name=\"base_cape_in_01_r\"),Transform=(Rotation=(X=0,Y=0,Z=0,W=1),Translation=(X=0,Y=2.5,Z=0),Scal ...(truncated)"
    WeightMinimum=0
    Weight=@"Delta Rotation Y_2".Delta Rotation
   )
   > Delta Rotation Y_3 = RigVMFunctionReferenceNode(Child=upperarm_l, Parent=upperarm_l_null, Source Maximum=45)
   > &RigUnit_ModifyTransforms(
    ItemToModify="((Item=(Type=Bone,Name=\"base_cape_out_01_l\"),Transform=(Rotation=(X=0,Y=0,Z=0.0175,W=0.9998),Translation=(X=0,Y=-3,Z=0),Scale3D=(X=1,Y=1,Z=1))),(Item=(Type=Bone,Name=\"base_cape_in_01_l\"),Transform=(Rotation=(X=0,Y=0,Z=0,W=1),Translation=(X=0,Y=-2.5,Z=0),Sc ...(truncated)"
    WeightMinimum=0
    Weight=@"Delta Rotation Y_3".Delta Rotation
   )
   > Delta Rotation Y = RigVMFunctionReferenceNode(Child=clavicle_r, Parent=clavicle_r_null, Source Maximum=-40)
   > &RigUnit_ModifyTransforms(
    ItemToModify="((Item=(Type=Bone,Name=\"base_cape_out_01_r\"),Transform=(Rotation=(X=0,Y=0,Z=0.1736,W=0.9848),Translation=(X=0,Y=0,Z=1),Scale3D=(X=1,Y=1,Z=1))),(Item=(Type=Bone,Name=\"base_cape_in_01_r\"),Transform=(Rotation=(X=0,Y=0,Z=0,W=1),Translation=(X=-2.5,Y=0,Z=1.5),S ...(truncated)"
    WeightMinimum=0
    Weight=@"Delta Rotation Y".Delta Rotation
   )
   > Delta Rotation Y_1 = RigVMFunctionReferenceNode(Child=clavicle_l, Parent=clavicle_l_null, Source Maximum=-40)
   > n2 = &RigUnit_ModifyTransforms(
    ItemToModify="((Item=(Type=Bone,Name=\"base_cape_out_01_l\"),Transform=(Rotation=(X=0,Y=0,Z=-0.1736,W=0.9848),Translation=(X=0,Y=0,Z=1),Scale3D=(X=1,Y=1,Z=1))),(Item=(Type=Bone,Name=\"base_cape_in_01_l\"),Transform=(Rotation=(X=0,Y=0,Z=0,W=1),Translation=(X=-2.5,Y=0,Z=1.5), ...(truncated)"
    WeightMinimum=0
    Weight=@"Delta Rotation Y_1".Delta Rotation
   )
   RigVMFunction_ControlFlowBranch = &RigVMFunction_ControlFlowBranch(ExecuteContext=@n2.ExecutePin, Condition=RigVMVariableNode() {variable=IsEmoting}.Value)
   Delta Rotation Y_4 = RigVMFunctionReferenceNode(Child=clavicle_r, Parent=clavicle_r_null, Source Maximum=-40, ExecuteContext=@RigVMFunction_ControlFlowBranch.True)
   > &RigUnit_ModifyTransforms(
    ItemToModify="((Item=(Type=Bone,Name=\"base_cape_out_01_r\"),Transform=(Rotation=(X=0,Y=0,Z=0.1305,W=0.9914),Translation=(X=0,Y=0,Z=0),Scale3D=(X=1,Y=1,Z=1))),(Item=(Type=Bone,Name=\"base_cape_in_01_r\"),Transform=(Rotation=(X=0,Y=0,Z=0,W=1),Translation=(X=0,Y=1.5,Z=2),Scal ...(truncated)"
    WeightMinimum=0
    Weight=@"Delta Rotation Y_4".Delta Rotation
   )
   > Delta Rotation Y_5 = RigVMFunctionReferenceNode(Child=clavicle_l, Parent=clavicle_l_null, Source Maximum=-40)
   > &RigUnit_ModifyTransforms(
    ItemToModify="((Item=(Type=Bone,Name=\"base_cape_out_01_l\"),Transform=(Rotation=(X=0,Y=0,Z=-0.1305,W=0.9914),Translation=(X=0,Y=0,Z=0),Scale3D=(X=1,Y=1,Z=1))),(Item=(Type=Bone,Name=\"base_cape_in_01_l\"),Transform=(Rotation=(X=0,Y=0,Z=0,W=1),Translation=(X=0,Y=-1.5,Z=2),Sc ...(truncated)"
    WeightMinimum=0
    Weight=@"Delta Rotation Y_5".Delta Rotation
   )
   > &RigUnit_ModifyTransforms(
    ItemToModify="((Item=(Type=Bone,Name=\"base_skirt_01_front_r\"),Transform=(Rotation=(X=0,Y=-0.0872,Z=0,W=0.9962),Translation=(X=0,Y=0,Z=0),Scale3D=(X=1,Y=1,Z=1))),(Item=(Type=Bone,Name=\"base_skirt_01_front_l\"),Transform=(Rotation=(X=0,Y=-0.0872,Z=0,W=0.9962),Translation=( ...(truncated)"
    Weight=1
    WeightMinimum=0
   )
   > &RigUnit_ParentConstraint(
    AdvancedSettings=(RotationOrderForFilter=XZY)
    Child=(Type=Bone,Name="base_bracelet_01_l")
    Parents="((Item=(Type=Bone,Name=\"thumb_01_l\"),Weight=0.0057),(Item=(Type=Bone,Name=\"lowerarm_twist_02_l\"),Weight=0.0072),(Item=(Type=Bone,Name=\"lowerarm_l\"),Weight=0.0169),(Item=(Type=Bone,Name=\"deform_wrist_dn_l\"),Weight=0.1062),(Item=(Type=Bone,Name=\"hand_l\"),Wei ...(truncated)"
    Weight=1
   )
   > &RigUnit_ParentConstraint(
    AdvancedSettings=(RotationOrderForFilter=XZY)
    Child=(Type=Bone,Name="base_bracelet_02_l")
    Parents="((Item=(Type=Bone,Name=\"hand_l\"),Weight=0.0039),(Item=(Type=Bone,Name=\"deform_wrist_up_l\"),Weight=0.0043),(Item=(Type=Bone,Name=\"lowerarm_twist_02_l\"),Weight=0.0927),(Item=(Type=Bone,Name=\"lowerarm_l\"),Weight=0.1211),(Item=(Type=Bone,Name=\"lowerarm_twist_0 ...(truncated)"
    Weight=1
   )
   > &RigUnit_ParentConstraint(
    AdvancedSettings=(RotationOrderForFilter=XZY)
    Child=(Type=Bone,Name="base_bracelet_03_l")
    Parents=((Item=(Type=Bone,Name="lowerarm_l"),Weight=0.2206),(Item=(Type=Bone,Name="lowerarm_twist_02_l"),Weight=0.2974),(Item=(Type=Bone,Name="lowerarm_twist_01_l"),Weight=0.482))
    Weight=1
   )
   RigVMFunctionReturnNode(ExecuteContext=@RigVMFunction_Sequence.B)
  ]}
 RigVMFunctionLibrary {nodes [
   Cape Setup = RigVMCollapseNode(ExecuteContext=())
   Delta Rotation X = RigVMCollapseNode(ExecuteContext=())
   Delta Rotation Y = RigVMCollapseNode(ExecuteContext=())
   Delta Rotation Z = RigVMCollapseNode(ExecuteContext=())
   Skirt Setup = RigVMCollapseNode(ExecuteContext=())
   Variables = RigVMCollapseNode(ExecuteContext=())
   rotationTransform = RigVMCollapseNode(ExecuteContext=(), Item=(Type=None,Name="None"))
  ]}
 RigVMFunction_Sequence_1_1_ContainedGraph {nodes [
   n1 = RigVMFunctionEntryNode()
   n2 = &RigVMFunction_Sequence(ExecuteContext=@n1.ExecuteContext)
   n3 = &RigVMFunction_Sequence(ExecuteContext=@n2.B)
   n4 = &RigVMFunction_Sequence(ExecuteContext=@n3.B)
   n5 = &RigVMFunction_Sequence(ExecuteContext=@n4.B)
   n6 = &RigVMFunction_Sequence(ExecuteContext=@n5.B)
   RigVMFunctionReturnNode(A=@n2.A, B=@n3.A, C=@n4.A, E=@n5.A, G=@n6.A, H=@n6.B)
  ]}
 RigVMFunction_Sequence_1_ContainedGraph {nodes [
   n1 = RigVMFunctionEntryNode()
   RigVMFunction_Sequence = &RigVMFunction_Sequence(ExecuteContext=@n1.ExecuteContext)
   n2 = &RigVMFunction_Sequence(ExecuteContext=@RigVMFunction_Sequence.B)
   RigVMFunctionReturnNode(A=@RigVMFunction_Sequence.A, B=@n2.A, C=@n2.B)
  ]}
 RigVMFunction_Sequence_2_1_ContainedGraph {nodes [
   n1 = RigVMFunctionEntryNode()
   n2 = &RigVMFunction_Sequence(ExecuteContext=@n1.ExecuteContext)
   n3 = &RigVMFunction_Sequence(ExecuteContext=@n2.B)
   n4 = &RigVMFunction_Sequence(ExecuteContext=@n3.B)
   RigVMFunctionReturnNode(C=@n4.A, D=@n4.B, A=@n2.A, B=@n3.A)
  ]}
 RigVMModel {nodes [
   RigUnit_BeginExecution = &RigUnit_BeginExecution()
   /Script/FortniteGame.RigUnit_FortDeformRig(
    Bool=(bUpperArm=true,bElbow=true,bWrist=true,bGroin=true,bGlute=true,bKnee=true,bPec=False,bBicep=False,bLat=False)
    Curve="(bLUpperArmUpAlpha=true,bRUpperArmUpAlpha=true,bLUpperArmFrAlpha=true,bRUpperArmFrAlpha=true,bLUpperArmBkAlpha=true,bRUpperArmBkAlpha=true,bLWristUpAlpha=False,bRWristUpAlpha=False,bLWristDnAlpha=False,bRWristDnAlpha=False,bLKneeAlpha=False,bRKneeAlpha=Fal ...(truncated)"
    Debug=(bUpperArmDebug=false,bWristDebug=False,bLatDebug=False,bGroinGluteDebug=False)
    Value="(UpperArmUp=(Rotation=(X=0.0,Y=-0.0,Z=0.0,W=1.0),Translation=(X=0.0,Y=0.0,Z=8.0),Scale3D=(X=0.70,Y=1.0,Z=1.0)),UpperArmFwdFr=(Rotation=(X=0.0,Y=-0.0,Z=0.0,W=1.0),Translation=(X=-3,Y=10.0,Z=0.0),Scale3D=(X=1,Y=1.0,Z=1.0)),UpperArmFwdBk=(Rotation=(X=0.0,Y=-0 ...(truncated)"
    ExecutePin=@RigUnit_BeginExecution.ExecutePin
   ) {is=struct RigVMUnitNode}
   > Variables = RigVMFunctionReferenceNode()
   > Skirt Setup = RigVMFunctionReferenceNode()
   > Cape Setup = RigVMFunctionReferenceNode()
  ]}
]
hierarchy [
 root {loc=<0 0 0>, rot=<0 0 0 1>}
 pelvis {parent=root, loc=<0 -0.5301 98.0571>, rot=<0.0127 0.707 -0.0127 -0.707>}
 spine_01 {parent=pelvis, loc=<6.1208 -0.9328 0>, rot=<0 0 0.034 -0.9994>}
 spine_02 {parent=spine_01, loc=<6.5997 -0.2095 0>, rot=<0 0 0.0238 -0.9997>}
 spine_03 {parent=spine_02, loc=<6.4914 -0.1237 0>, rot=<0 0 0.0149 -0.9999>}
 spine_04 {parent=spine_03, loc=<6.9598 0.0794 0>, rot=<0 0 -0.0132 -0.9999>}
 spine_05 {parent=spine_04, loc=<6.9987 0 0>, rot=<0 0 -0.0015 -1>}
 attach_cape {parent=spine_05, loc=<9.017 -11.714 0>, rot=<0.0616 -0.7044 -0.0616 -0.7044>}
 cape {parent=attach_cape, type=Null, loc=<0 2.1171 -4.1164>, rot=<0.0549 0 0 0.9985>}
 neck_01 {parent=spine_05, loc=<11.3138 0.1529 0>, rot=<0 0 -0.2003 -0.9797>}
 neck_02 {parent=neck_01, loc=<4.5627 0 0>, rot=<0 0 0.0504 -0.9987>}
 head {parent=neck_02, loc=<5.3914 0 0>, rot=<0 0 0.115 -0.9934>}
 attach_backpack {parent=spine_05, loc=<-0.123 -13.473 0>, rot=<-0.0093 -0.707 0.0093 -0.707>}
 Backpack_BR {parent=attach_backpack, type=Null, loc=<0 2.0126 -2.5466>, rot=<-0.1408 0 0 0.99>}
 clavicle_l {parent=spine_05, loc=<5.4831 -0.1053 -3.2946>, rot=<0.0465 -0.7289 -0.009 -0.6829>}
 upperarm_l {parent=clavicle_l, loc=<9.8658 -1.2136 -0.4105>, rot=<0.0164 -0.3245 -0.0269 -0.9454>}
 lowerarm_l {parent=upperarm_l, loc=<26.465 -0.8345 0.259>, rot=<0.0036 0.0161 -0.2298 -0.9731>}
 hand_l {parent=lowerarm_l, loc=<22.7355 -0.3203 -0.6105>, rot=<-0.6784 -0.0223 0.0735 -0.7307>}
 thumb_01_l {parent=hand_l, loc=<1.3118 -0.3191 -2.459>, rot=<0.6546 -0.4042 0.0664 -0.6354>}
 thumb_02_l {parent=thumb_01_l, loc=<5.9336 -0.0448 0.0401>, rot=<-0.0037 0.0213 0.298 -0.9543>}
 thumb_03_l {parent=thumb_02_l, loc=<2.2223 0.3682 -0.0907>, rot=<0.0118 -0.013 0.0066 -0.9998>}
 index_metacarpal_l {parent=hand_l, loc=<1.6535 0.0183 -2.96>, rot=<0.0158 -0.0124 -0.0173 -0.9996>}
 index_01_l {parent=index_metacarpal_l, loc=<8.3008 0.0357 -0.0524>, rot=<-0.0044 0.0485 0.0924 -0.9945>}
 index_02_l {parent=index_01_l, loc=<3.3587 -0.3004 0.0452>, rot=<-0.002 0.0027 0.3225 -0.9466>}
 index_03_l {parent=index_02_l, loc=<2.5578 -0.123 -0.0252>, rot=<-0.002 0.0067 0.2558 -0.9667>}
 middle_metacarpal_l {parent=hand_l, loc=<2.1856 -0.1725 -0.5984>, rot=<0.006 0.0168 -0.0329 -0.9993>}
 middle_01_l {parent=middle_metacarpal_l, loc=<7.3275 -0.2447 0.0129>, rot=<-0.0208 0.0042 0.0858 -0.9961>}
 middle_02_l {parent=middle_01_l, loc=<4.0123 0.1401 0.1175>, rot=<-0.0218 0.0258 0.3512 -0.9357>}
 middle_03_l {parent=middle_02_l, loc=<2.6077 0.0534 0.1373>, rot=<0.0082 0.0188 0.1708 -0.9851>}
 ring_metacarpal_l {parent=hand_l, loc=<2.2074 -0.3625 1.3638>, rot=<-0.0284 0.0722 -0.0458 -0.9959>}
 ring_01_l {parent=ring_metacarpal_l, loc=<7.0295 -0.6197 -0.1446>, rot=<-0.0489 -0.0252 0.1892 -0.9804>}
 ring_02_l {parent=ring_01_l, loc=<3.6174 0.4647 -0.1027>, rot=<0 0 0.2806 -0.9598>}
 ring_03_l {parent=ring_02_l, loc=<2.4152 0.114 0.0199>, rot=<0 0 0.1189 -0.9929>}
 pinky_metacarpal_l {parent=hand_l, loc=<1.7616 -0.9753 3.3814>, rot=<-0.0397 0.0814 -0.069 -0.9935>}
 pinky_01_l {parent=pinky_metacarpal_l, loc=<7.1709 -0.4831 -0.1094>, rot=<-0.0654 -0.0035 0.2242 -0.9723>}
 pinky_02_l {parent=pinky_01_l, loc=<2.8274 0.0893 -0.1043>, rot=<0.0037 0.0011 0.2854 -0.9584>}
 pinky_03_l {parent=pinky_02_l, loc=<2.2046 0.0307 -0.0078>, rot=<-0.0002 0.0098 0.1445 -0.9895>}
 deform_wrist_dn_base_l {parent=hand_l, loc=<0 -1.5326 0>, rot=<0 0 0 -1>}
 deform_wrist_dn_l {parent=deform_wrist_dn_base_l, loc=<0 0 0>, rot=<0 0 0 1>}
 deform_wrist_up_base_l {parent=hand_l, loc=<0 2.3009 0>, rot=<0 0 0 -1>}
 deform_wrist_up_l {parent=deform_wrist_up_base_l, loc=<0 0 0>, rot=<0 0 0 1>}
 lowerarm_twist_02_l {parent=lowerarm_l, loc=<10.6707 0 0>, rot=<0 0 0 1>}
 lowerarm_twist_01_l {parent=lowerarm_l, loc=<18.316 0 0>, rot=<0 0 0 1>}
 deform_elbow_in_base_l {parent=lowerarm_l, loc=<0.1834 3.6062 -0.2572>, rot=<0 0 0 -1>}
 deform_elbow_in_l {parent=deform_elbow_in_base_l, loc=<0 0 0>, rot=<0 0 0 1>}
 deform_elbow_out_base_l {parent=lowerarm_l, loc=<0 -4 0>, rot=<0 0 0 1>}
 deform_elbow_out_l {parent=deform_elbow_out_base_l, loc=<0 0 0>, rot=<0 0 0 1>}
 base_bracelet_01_l {parent=lowerarm_l, loc=<19.1609 -0.169 -0.1369>, rot=<0 0 0 1>}
 dyn_bracelet_01_l {parent=base_bracelet_01_l, loc=<0 0 0>, rot=<0 0 0 1>}
 base_bracelet_02_l {parent=lowerarm_l, loc=<15.5181 -0.0805 0.1545>, rot=<0 0 0 1>}
 dyn_bracelet_02_l {parent=base_bracelet_02_l, loc=<0 0 0>, rot=<0 0 0 1>}
 base_bracelet_03_l {parent=lowerarm_l, loc=<12.2782 0.2727 0.2671>, rot=<0 0 0 1>}
 dyn_bracelet_03_l {parent=base_bracelet_03_l, loc=<0 0 0>, rot=<0 0 0 1>}
 upperarm_twist_01_l {parent=upperarm_l, loc=<7.2135 -0.1134 0>, rot=<0 0 0 1>}
 deform_upperarm_bk_base_l {parent=upperarm_twist_01_l, loc=<-6.4053 -4.3746 0.314>, rot=<0 0 0 -1>}
 deform_upperarm_bk_l {parent=deform_upperarm_bk_base_l, loc=<0 0 0>, rot=<0 0 0 1>}
 base_cape_out_01_l {parent=deform_upperarm_bk_base_l, loc=<4.2481 -0.7412 0.8712>, rot=<-0.6015 -0.1759 0.3717 -0.6849>}
 dyn_cape_out_01_l {parent=base_cape_out_01_l, loc=<0 0 0>, rot=<0 0 0 -1>}
 dyn_cape_out_02_l {parent=dyn_cape_out_01_l, loc=<10.7003 -0.0001 -0.0002>, rot=<0 0 0 -1>}
 dyn_cape_out_03_l {parent=dyn_cape_out_02_l, loc=<10.6994 0 0.0001>, rot=<0 0 0 -1>}
 dyn_cape_out_04_l {parent=dyn_cape_out_03_l, loc=<10.6994 0.0001 0.0002>, rot=<0 0 0 -1>}
 deform_upperarm_fr_base_l {parent=upperarm_twist_01_l, loc=<-5.6312 5.092 0.3354>, rot=<0 0 0 -1>}
 deform_upperarm_fr_l {parent=deform_upperarm_fr_base_l, loc=<0 0 0>, rot=<0 0 0 1>}
 deform_upperarm_up_base_l {parent=upperarm_twist_01_l, loc=<-8.891 0.4825 3.9517>, rot=<0 0 0 -1>}
 deform_upperarm_up_l {parent=deform_upperarm_up_base_l, loc=<0 0 0>, rot=<0 0 0 1>}
 upperarm_twist_02_l {parent=upperarm_l, loc=<19.7998 -0.1134 0>, rot=<0 0 0 1>}
 upperarm_l_null {parent=clavicle_l, type=Null, loc=<9.8658 -1.2136 -0.4105>, rot=<0.0164 -0.3245 -0.0269 -0.9454>}
 upperarm_l_Hold {parent=clavicle_l, loc=<10.782 0 0>, rot=<0 -0.0266 0 -0.9996>}
 clavicle_r {parent=spine_05, loc=<5.4831 -0.1053 3.2946>, rot=<0.7289 0.0465 -0.6829 0.009>}
 upperarm_r {parent=clavicle_r, loc=<-9.8658 1.2136 0.4105>, rot=<0.0164 -0.3245 -0.0269 -0.9454>}
 lowerarm_r {parent=upperarm_r, loc=<-26.465 0.8345 -0.259>, rot=<0.0036 0.0161 -0.2298 -0.9731>}
 hand_r {parent=lowerarm_r, loc=<-22.7355 0.3203 0.6106>, rot=<-0.6784 -0.0223 0.0735 -0.7307>}
 thumb_01_r {parent=hand_r, loc=<-1.3118 0.3191 2.459>, rot=<0.6546 -0.4042 0.0664 -0.6354>}
 thumb_02_r {parent=thumb_01_r, loc=<-5.9336 0.0448 -0.0401>, rot=<-0.0037 0.0213 0.298 -0.9543>}
 thumb_03_r {parent=thumb_02_r, loc=<-2.2223 -0.3682 0.0907>, rot=<0.0118 -0.013 0.0066 -0.9998>}
 index_metacarpal_r {parent=hand_r, loc=<-1.6535 -0.0183 2.96>, rot=<0.0158 -0.0124 -0.0173 -0.9996>}
 index_01_r {parent=index_metacarpal_r, loc=<-8.3008 -0.0357 0.0524>, rot=<-0.0044 0.0485 0.0924 -0.9945>}
 index_02_r {parent=index_01_r, loc=<-3.3587 0.3004 -0.0452>, rot=<-0.002 0.0027 0.3225 -0.9466>}
 index_03_r {parent=index_02_r, loc=<-2.5578 0.123 0.0252>, rot=<-0.002 0.0067 0.2558 -0.9667>}
 middle_metacarpal_r {parent=hand_r, loc=<-2.1856 0.1725 0.5984>, rot=<0.006 0.0168 -0.0329 -0.9993>}
 middle_01_r {parent=middle_metacarpal_r, loc=<-7.3275 0.2447 -0.0129>, rot=<-0.0208 0.0042 0.0858 -0.9961>}
 middle_02_r {parent=middle_01_r, loc=<-4.0123 -0.1401 -0.1175>, rot=<-0.0218 0.0258 0.3512 -0.9357>}
 middle_03_r {parent=middle_02_r, loc=<-2.6077 -0.0534 -0.1373>, rot=<0.0082 0.0188 0.1708 -0.9851>}
 ring_metacarpal_r {parent=hand_r, loc=<-2.2074 0.3625 -1.3638>, rot=<-0.0284 0.0722 -0.0458 -0.9959>}
 ring_01_r {parent=ring_metacarpal_r, loc=<-7.0295 0.6197 0.1446>, rot=<-0.0489 -0.0252 0.1892 -0.9804>}
 ring_02_r {parent=ring_01_r, loc=<-3.6174 -0.4647 0.1027>, rot=<0 0 0.2806 -0.9598>}
 ring_03_r {parent=ring_02_r, loc=<-2.4152 -0.114 -0.0199>, rot=<0 0 0.1189 -0.9929>}
 pinky_metacarpal_r {parent=hand_r, loc=<-1.7616 0.9753 -3.3814>, rot=<-0.0397 0.0814 -0.069 -0.9935>}
 pinky_01_r {parent=pinky_metacarpal_r, loc=<-7.1709 0.4831 0.1094>, rot=<-0.0654 -0.0035 0.2242 -0.9723>}
 pinky_02_r {parent=pinky_01_r, loc=<-2.8274 -0.0893 0.1043>, rot=<0.0037 0.0011 0.2854 -0.9584>}
 pinky_03_r {parent=pinky_02_r, loc=<-2.2046 -0.0307 0.0078>, rot=<-0.0002 0.0098 0.1445 -0.9895>}
 deform_wrist_dn_base_r {parent=hand_r, loc=<-0.0001 1.5328 0>, rot=<0 0 0 -1>}
 deform_wrist_dn_r {parent=deform_wrist_dn_base_r, loc=<0 0 0>, rot=<0 0 0 1>}
 deform_wrist_up_base_r {parent=hand_r, loc=<-0.0002 -2.3007 -0.0001>, rot=<0 0 0 -1>}
 deform_wrist_up_r {parent=deform_wrist_up_base_r, loc=<0 0 0>, rot=<0 0 0 1>}
 lowerarm_twist_02_r {parent=lowerarm_r, loc=<-10.6707 0 0>, rot=<0 0 0 -1>}
 lowerarm_twist_01_r {parent=lowerarm_r, loc=<-18.316 0 0>, rot=<0 0 0 -1>}
 deform_elbow_in_base_r {parent=lowerarm_r, loc=<-0.1834 -3.6062 0.2572>, rot=<0 0 0 -1>}
 deform_elbow_in_r {parent=deform_elbow_in_base_r, loc=<0 0 0>, rot=<0 0 0 -1>}
 deform_elbow_out_base_r {parent=lowerarm_r, loc=<0 4 -0.0001>, rot=<0 0 0 -1>}
 deform_elbow_out_r {parent=deform_elbow_out_base_r, loc=<0 0 0>, rot=<0 0 0 1>}
 upperarm_twist_01_r {parent=upperarm_r, loc=<-7.2135 0.1134 0>, rot=<0 0 0 -1>}
 deform_upperarm_bk_base_r {parent=upperarm_twist_01_r, loc=<6.4053 4.3746 -0.314>, rot=<0 0 0 -1>}
 deform_upperarm_bk_r {parent=deform_upperarm_bk_base_r, loc=<0 0 0>, rot=<0 0 0 1>}
 base_cape_out_01_r {parent=deform_upperarm_bk_base_r, loc=<-4.2482 0.7413 -0.8711>, rot=<-0.3717 -0.6849 -0.6015 0.1759>}
 dyn_cape_out_01_r {parent=base_cape_out_01_r, loc=<0 0 0>, rot=<0 0 0 -1>}
 dyn_cape_out_02_r {parent=dyn_cape_out_01_r, loc=<10.6998 0 0>, rot=<0 0 0 1>}
 dyn_cape_out_03_r {parent=dyn_cape_out_02_r, loc=<10.6998 0 0>, rot=<0 0 0 1>}
 dyn_cape_out_04_r {parent=dyn_cape_out_03_r, loc=<10.6998 0 0>, rot=<0 0 0 1>}
 deform_upperarm_fr_base_r {parent=upperarm_twist_01_r, loc=<5.6312 -5.092 -0.3354>, rot=<0 0 0 -1>}
 deform_upperarm_fr_r {parent=deform_upperarm_fr_base_r, loc=<0 0 0>, rot=<0 0 0 1>}
 deform_upperarm_up_base_r {parent=upperarm_twist_01_r, loc=<8.891 -0.4825 -3.9517>, rot=<0 0 0 -1>}
 deform_upperarm_up_r {parent=deform_upperarm_up_base_r, loc=<0 0 0>, rot=<0 0 0 1>}
 upperarm_twist_02_r {parent=upperarm_r, loc=<-19.7998 0.1134 0>, rot=<0 0 0 -1>}
 upperarm_r_null {parent=clavicle_r, type=Null, loc=<-9.8658 1.2136 0.4105>, rot=<0.0164 -0.3245 -0.0269 -0.9454>}
 upperarm_r_Hold {parent=clavicle_r, loc=<16.2503 0 0>, rot=<0 0 0 1>}
 neck_01_null {parent=spine_05, type=Null, loc=<5.4237 1.0981 -2.7084>, rot=<-0.1326 0.003 0.1827 0.9742>}
 spine_05_null {parent=spine_05, type=Null, loc=<0 0 0>, rot=<0 0 0 1>}
 cape_root {parent=spine_05, loc=<4.8762 -7.7236 0>, rot=<0.0284 -0.7065 0.7065 -0.0284>}
 base_cape_in_01_r {parent=cape_root, loc=<-0.1208 -7.6183 0.2253>, rot=<0 0.1466 0 -0.9892>}
 dyn_cape_in_01_r {parent=base_cape_in_01_r, loc=<0 0 0>, rot=<0 0 0 1>}
 dyn_cape_in_02_r {parent=dyn_cape_in_01_r, loc=<11.5607 0 0>, rot=<0 0 0 1>}
 dyn_cape_in_03_r {parent=dyn_cape_in_02_r, loc=<11.5607 0 0>, rot=<0 0 0 1>}
 dyn_cape_in_04_r {parent=dyn_cape_in_03_r, loc=<11.5607 0 0>, rot=<0 0 0 1>}
 base_cape_in_01_l {parent=cape_root, loc=<-0.1204 7.6183 0.2252>, rot=<0 0.1466 0 -0.9892>}
 dyn_cape_in_01_l {parent=base_cape_in_01_l, loc=<0 0 0>, rot=<0 0 0 1>}
 dyn_cape_in_02_l {parent=dyn_cape_in_01_l, loc=<11.5606 0 0>, rot=<0 0 0 1>}
 dyn_cape_in_03_l {parent=dyn_cape_in_02_l, loc=<11.5606 0 0>, rot=<0 0 0 1>}
 dyn_cape_in_04_l {parent=dyn_cape_in_03_l, loc=<11.5606 0 0>, rot=<0 0 0 1>}
 clavicle_l_null {parent=spine_05, type=Null, loc=<5.483 -0.1053 -3.2946>, rot=<0.0465 -0.7289 -0.009 -0.6829>}
 clavicle_r_null {parent=spine_05, type=Null, loc=<5.483 -0.1053 3.2946>, rot=<0.7289 0.0465 -0.6829 0.009>}
 deform_pec_aim_base_l {parent=spine_04, loc=<3.2456 9.3309 -1.2237>, rot=<0.0603 -0.4033 0.3318 -0.8507>}
 deform_pec_aim_l {parent=deform_pec_aim_base_l, loc=<0 0 0>, rot=<0 0 0 1>}
 deform_pec_base_l {parent=deform_pec_aim_l, loc=<9.3665 0 0>, rot=<0 0 0 1>}
 deform_pec_l {parent=deform_pec_base_l, loc=<0 0 0>, rot=<0 0 0 -1>}
 deform_pec_aim_base_r {parent=spine_04, loc=<3.2452 9.3309 1.2237>, rot=<-0.4033 -0.0603 0.8507 0.3318>}
 deform_pec_aim_r {parent=deform_pec_aim_base_r, loc=<0 0 0>, rot=<0 0 0 -1>}
 deform_pec_base_r {parent=deform_pec_aim_r, loc=<-9.3667 -0.0003 -0.0004>, rot=<0 0 0 1>}
 deform_pec_r {parent=deform_pec_base_r, loc=<0 0 0>, rot=<0 0 0 1>}
 base_chest_l {parent=spine_04, loc=<5.6419 6.7389 -6.63>, rot=<0.5 0.5 0.5 0.5>}
 dyn_chest_l {parent=base_chest_l, loc=<0 0 0>, rot=<0 0 0 1>}
 base_chest_r {parent=spine_04, loc=<5.6419 6.7389 6.63>, rot=<0.5 0.5 0.5 0.5>}
 dyn_chest_r {parent=base_chest_r, loc=<0 0 0>, rot=<0 0 0 1>}
 thigh_l {parent=pelvis, loc=<-1.0855 -0.2109 -8.7053>, rot=<0.0009 0.0355 0.0155 -0.9993>}
 calf_l {parent=thigh_l, loc=<-42.5284 -0.0019 -0.7297>, rot=<0.0006 -0.0161 -0.0391 -0.9991>}
 foot_l {parent=calf_l, loc=<-46.7401 -2.5014 -0.592>, rot=<0.0427 -0.0238 0.0353 -0.9982>}
 ball_l {parent=foot_l, loc=<-3.5658 10.8066 -0.1544>, rot=<0.0001 -0.0001 -0.7063 -0.7079>}
 calf_twist_02_l {parent=calf_l, loc=<-19.1546 -1.7943 -0.2638>, rot=<0.0237 0.0095 -0.0035 -0.9997>}
 calf_twist_01_l {parent=calf_l, loc=<-35.8483 -1.9197 -0.5794>, rot=<0.0237 0.0095 -0.0035 -0.9997>}
 deform_knee_in_base_l {parent=calf_l, loc=<0.8031 -5.0231 0.2456>, rot=<0 0 0 -1>}
 deform_knee_in_l {parent=deform_knee_in_base_l, loc=<0 0 0>, rot=<0 0 0 1>}
 deform_knee_out_base_l {parent=calf_l, loc=<0 4 0>, rot=<0 0 0 -1>}
 deform_knee_out_l {parent=deform_knee_out_base_l, loc=<0 0 0>, rot=<0 0 0 1>}
 thigh_twist_01_l {parent=thigh_l, loc=<-27.952 0 0>, rot=<0.0013 0 0 -1>}
 deform_groin_base_l {parent=thigh_l, loc=<0.4801 5.6776 -0.2486>, rot=<0 0 0 -1>}
 deform_groin_l {parent=deform_groin_base_l, loc=<0 0 0>, rot=<0 0 0 1>}
 calf_l_null {parent=thigh_l, type=Null, loc=<-42.5284 -0.0019 -0.7297>, rot=<0.0006 -0.0161 -0.0391 -0.9991>}
 thigh_r {parent=pelvis, loc=<-1.0855 -0.2109 8.7053>, rot=<0.0355 -0.0009 0.9993 0.0155>}
 calf_r {parent=thigh_r, loc=<42.5284 0.0019 0.7297>, rot=<0.0006 -0.0161 -0.0391 -0.9991>}
 foot_r {parent=calf_r, loc=<46.7401 2.5014 0.592>, rot=<0.0427 -0.0238 0.0353 -0.9982>}
 ball_r {parent=foot_r, loc=<3.5658 -10.8066 0.1544>, rot=<0.0001 -0.0001 -0.7063 -0.7079>}
 calf_twist_02_r {parent=calf_r, loc=<19.1546 1.7943 0.2638>, rot=<0.0237 0.0095 -0.0035 -0.9997>}
 calf_twist_01_r {parent=calf_r, loc=<35.8483 1.9197 0.5794>, rot=<0.0237 0.0095 -0.0035 -0.9997>}
 deform_knee_in_base_r {parent=calf_r, loc=<-0.8031 5.023 -0.2457>, rot=<0 0 0 -1>}
 deform_knee_in_r {parent=deform_knee_in_base_r, loc=<0 0 0>, rot=<0 0 0 1>}
 deform_knee_out_base_r {parent=calf_r, loc=<0 -4 0>, rot=<0 0 0 -1>}
 deform_knee_out_r {parent=deform_knee_out_base_r, loc=<0 0 0>, rot=<0 0 0 1>}
 thigh_twist_01_r {parent=thigh_r, loc=<27.952 0 0>, rot=<0.0013 0 0 -1>}
 deform_groin_base_r {parent=thigh_r, loc=<-0.4801 -5.6776 0.2486>, rot=<0 0 0 -1>}
 deform_groin_r {parent=deform_groin_base_r, loc=<0 0 0>, rot=<0 0 0 1>}
 calf_r_null {parent=thigh_r, type=Null, loc=<42.5284 0.0019 0.7297>, rot=<0.0006 -0.0161 -0.0391 -0.9991>}
 deform_glute_base_l {parent=pelvis, loc=<-7.3466 -9.0213 -9.1245>, rot=<0.0172 0.031 -0.4477 -0.8935>}
 deform_glute_l {parent=deform_glute_base_l, loc=<0 0 0>, rot=<0 0 0 1>}
 deform_glute_base_r {parent=pelvis, loc=<-7.3464 -9.0213 9.1245>, rot=<0.031 -0.0172 0.8935 -0.4477>}
 deform_glute_r {parent=deform_glute_base_r, loc=<0 0 0>, rot=<0 0 0 1>}
 base_skirt_01_bck {parent=pelvis, loc=<-4.1239 -11.7668 -0.7526>, rot=<0.0762 -0.703 0.703 -0.0762>}
 dyn_skirt_01_bck {parent=base_skirt_01_bck, loc=<0 0 0>, rot=<0 0 0 1>}
 dyn_skirt_02_bck {parent=dyn_skirt_01_bck, loc=<16.3897 0 0>, rot=<0 -0.0037 0 -1>}
 dyn_skirt_03_bck {parent=dyn_skirt_02_bck, loc=<16.7296 0 0>, rot=<0 -0.0143 0 -0.9999>}
 dyn_skirt_04_bck {parent=dyn_skirt_03_bck, loc=<10.8228 0 0>, rot=<0 0 0 1>}
 dyn_skirt_05_bck {parent=dyn_skirt_04_bck, loc=<15.3888 0 0>, rot=<0 0 0 1>}
 base_skirt_01_bck_l {parent=pelvis, loc=<-3.8203 -9.2243 -10.2928>, rot=<0.0421 -0.8624 0.4908 -0.1164>}
 dyn_skirt_01_bck_l {parent=base_skirt_01_bck_l, loc=<0 0 0>, rot=<0 0 0 1>}
 dyn_skirt_02_bck_l {parent=dyn_skirt_01_bck_l, loc=<16.3897 0 0>, rot=<-0.0004 -0.0026 0.0005 -1>}
 dyn_skirt_03_bck_l {parent=dyn_skirt_02_bck_l, loc=<15.5729 0 0>, rot=<-0.002 -0.015 0.0032 -0.9999>}
 dyn_skirt_04_bck_l {parent=dyn_skirt_03_bck_l, loc=<11.271 0 0>, rot=<0.0022 0.0149 -0.0005 -0.9999>}
 dyn_skirt_05_bck_l {parent=dyn_skirt_04_bck_l, loc=<17.1382 0 0>, rot=<0 0 0 1>}
 base_skirt_01_bck_r {parent=pelvis, loc=<-3.8197 -9.2171 10.305>, rot=<0.1139 -0.4914 0.8622 -0.0466>}
 dyn_skirt_01_bck_r {parent=base_skirt_01_bck_r, loc=<0 0 0>, rot=<0 0 0 1>}
 dyn_skirt_02_bck_r {parent=dyn_skirt_01_bck_r, loc=<16.3897 0.0001 0>, rot=<0.0004 -0.0026 -0.0005 -1>}
 dyn_skirt_03_bck_r {parent=dyn_skirt_02_bck_r, loc=<15.5729 0 0>, rot=<0.002 -0.015 -0.0032 -0.9999>}
 dyn_skirt_04_bck_r {parent=dyn_skirt_03_bck_r, loc=<11.2709 0 0>, rot=<-0.0022 0.0149 0.0005 -0.9999>}
 dyn_skirt_05_bck_r {parent=dyn_skirt_04_bck_r, loc=<17.1383 0 0>, rot=<0 0 0 1>}
 base_skirt_01_l {parent=pelvis, loc=<-3.1219 -0.5761 -14.3723>, rot=<0.0088 -0.9921 -0.0011 -0.1251>}
 dyn_skirt_01_l {parent=base_skirt_01_l, loc=<0 0 0>, rot=<0 0 0 1>}
 dyn_skirt_02_l {parent=dyn_skirt_01_l, loc=<16.0432 0 0>, rot=<0 -0.0129 -0.0002 -0.9999>}
 dyn_skirt_03_l {parent=dyn_skirt_02_l, loc=<16.0994 0 0>, rot=<0 -0.0171 -0.0002 -0.9999>}
 dyn_skirt_04_l {parent=dyn_skirt_03_l, loc=<11.2774 0 0>, rot=<-0.0001 -0.0435 -0.0005 -0.9991>}
 dyn_skirt_05_l {parent=dyn_skirt_04_l, loc=<16.2575 0 0>, rot=<0 0 0 1>}
 base_skirt_01_r {parent=pelvis, loc=<-3.1219 -0.5761 14.3724>, rot=<0.1251 0.0011 0.9921 -0.0088>}
 dyn_skirt_01_r {parent=base_skirt_01_r, loc=<0 0 0>, rot=<0 0 0 1>}
 dyn_skirt_02_r {parent=dyn_skirt_01_r, loc=<16.0432 0 -0.0001>, rot=<0 -0.0129 0.0002 -0.9999>}
 dyn_skirt_03_r {parent=dyn_skirt_02_r, loc=<16.0994 0 0.0001>, rot=<0 -0.0171 0.0002 -0.9999>}
 dyn_skirt_04_r {parent=dyn_skirt_03_r, loc=<11.2775 0 -0.0001>, rot=<0.0001 -0.0435 0.0005 -0.9991>}
 dyn_skirt_05_r {parent=dyn_skirt_04_r, loc=<16.2575 0 0>, rot=<0 0 0 1>}
 thigh_l_null {parent=pelvis, type=Null, loc=<-1.0855 -0.2109 -8.7053>, rot=<0.0009 0.0355 0.0155 -0.9993>}
 thigh_r_null {parent=pelvis, type=Null, loc=<-1.0855 -0.2109 8.7053>, rot=<0.0355 -0.0009 0.9993 0.0155>}
 base_belt {parent=pelvis, loc=<0 0 0>, rot=<0 0.7071 0.7071 0>}
 dyn_belt {parent=base_belt, loc=<0 0 0>, rot=<0 0 0 1>}
 base_skirt_01_front_r {parent=pelvis, loc=<-4.7711 7.304 6.0085>, rot=<0.0428 0.7058 0.7058 0.0428>}
 dyn_skirt_01_front_r {parent=base_skirt_01_front_r, loc=<0 0 0>, rot=<0 0 0 1>}
 dyn_skirt_02_front_r {parent=dyn_skirt_01_front_r, loc=<16.1108 0 0>, rot=<0 -0.026 0 -0.9997>}
 dyn_skirt_03_front_r {parent=dyn_skirt_02_front_r, loc=<16.0646 0 0>, rot=<0 -0.024 0 -0.9997>}
 dyn_skirt_04_front_r {parent=dyn_skirt_03_front_r, loc=<10.782 0 0>, rot=<0 -0.0266 0 -0.9996>}
 dyn_skirt_05_front_r {parent=dyn_skirt_04_front_r, loc=<16.2503 0 0>, rot=<0 0 0 1>}
 base_skirt_01_front_l {parent=pelvis, loc=<-4.7711 7.304 -6.0085>, rot=<-0.0428 -0.7058 -0.7058 -0.0428>}
 dyn_skirt_01_front_l {parent=base_skirt_01_front_l, loc=<0 0 0>, rot=<0 0 0 1>}
 dyn_skirt_02_front_l {parent=dyn_skirt_01_front_l, loc=<16.1108 0 0>, rot=<0 -0.026 0 -0.9997>}
 dyn_skirt_03_front_l {parent=dyn_skirt_02_front_l, loc=<16.0646 0 0>, rot=<0 -0.024 0 -0.9997>}
 dyn_skirt_04_front_l {parent=dyn_skirt_03_front_l, loc=<10.782 0 0>, rot=<0 -0.0266 0 -0.9996>}
 dyn_skirt_05_front_l {parent=dyn_skirt_04_front_l, loc=<16.2503 0 0>, rot=<0 0 0 1>}
 spine_05_null_2 {parent=pelvis, type=Null, loc=<32.9717 -4.2232 0>, rot=<0 0 0.058 -0.9983>}
 pelvis_null {parent=root, type=Null, loc=<0 -0.5301 98.0571>, rot=<0.0127 0.707 -0.0127 -0.707>}
]
variables [
 ArmUpSkydiveL {type=double, default=0}
 ArmUpSkydiveR {type=double, default=0}
 Calf_L_Rotation {type=double, default=0}
 Calf_R_Rotation {type=double, default=0}
 IsEmoting {type=bool, default=False}
 IsInAir {type=bool, default=False}
 KneeSliding {type=bool, default=False}
 NotInAir {type=double, default=0}
 PelvisToGround {type=double, default=0}
 SpineRotRight {type=double, default=0}
 TacticalSprinting {type=bool, default=False}
 Thigh_FrontSkirt_l {type=double, default=0}
 Thigh_FrontSkirt_r {type=double, default=0}
 Thigh_L_BackRotation {type=double, default=0}
 Thigh_L_In_SideRotation {type=double, default=0}
 Thigh_L_SideRotation {type=double, default=0}
 Thigh_R_BackRotation {type=double, default=0}
 Thigh_R_SideRotation {type=double, default=0}
 glider {type=bool, default=False}
 headLookUp {type=double, default=0}
 inFloating {type=double, default=0}
 isCrouching {type=bool, default=False}
 isDBNO {type=bool, default=False}
 isMoving {type=double, default=0}
 isSkydiving {type=double, default=0}
 openParachute {type=double, default=0}
 skydiveThigBackL {type=double, default=0}
 skydiveThigBackR {type=double, default=0}
 spineRotBack {type=double, default=0}
]
